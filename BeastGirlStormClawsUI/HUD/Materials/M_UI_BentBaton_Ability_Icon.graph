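# Translucent UI material for the Bent Baton claws ability icon on the HUD, with fill, attack-ready and cooldown states.
# A remapped Fill parameter steps against vertical UV for the fill mask, while rotated UVs blend a four-color gradient that replaces the filling color when ready.
# An anti-aliased SDF backing circle, MSDF icon and offset icon shadow form the masks.
# IsCooldown switches how these masks composite into the saturated results.
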Material /BeastGirlStormClawsUI/HUD/Materials/M_UI_BentBaton_Ability_Icon
graph {
 properties {bAutomaticallySetUsageInEditor=false, blendMode=BLEND_Translucent, materialDomain=MD_UI, usedShadingModels=Default Lit}
 nodes [
  n1 = ComponentMask(None=TextureCoordinate()) {g=true}
  n2 = ScalarParameter() {parameterName=Fill}
  n3 = Constant2Vector() {g=1}
  n4 = Constant2Vector() {g=0.98, r=0.02}
  NamedRerouteDeclaration(None=Step(
    Y=OneMinus(None=@n1)
    X=MaterialFunctionCall(Input=@n2, Input Low=@n3.R, Input High=@n3.R, Target Low=@n4.R, Target High=@n4.R) {materialFunction=&RemapValueRange}.Result
   ) {constX=0.5}) {name=FillMask, nodeColor=<rgba 0.3359 0 1 1>}
  n5 = TextureCoordinate()
  n6 = Constant() {r=0.87}
  n7 = Multiply(A=ComponentMask(None=MaterialFunctionCall(UVs=@n5, "Rotation Angle (0-1)"=@n6) {materialFunction=&CustomRotator}.Rotated Values) {r=true}) {constA=1, constB=3}
  n8 = Subtract(A=@n7)
  n9 = LinearInterpolate(
   A=LinearInterpolate(
    A=VectorParameter() {parameterName=ColorA, defaultValue=<rgba 0.9734 0.552 1 1>}.RGB
    B=VectorParameter() {parameterName=ColorB, defaultValue=<rgba 0.9301 0.2542 1 1>}.RGB
    Alpha=Saturate(None=@n7)
   )
   B=VectorParameter() {parameterName=ColorC, defaultValue=<rgba 0.6724 0.0273 1 1>}.RGB
   Alpha=Saturate(None=@n8)
  )
  n10 = VectorParameter() {parameterName=ColorD, defaultValue=<rgba 0.1779 0 0.3813 1>}
  n11 = Saturate(None=Subtract(A=@n8))
  NamedRerouteDeclaration(None=LinearInterpolate(
    A=VectorParameter() {parameterName=ColorWhileFIlling, defaultValue=<rgba 1 1 1 1>}.RGB
    B=LinearInterpolate(A=@n9, B=@n10.RGB, Alpha=@n11)
    Alpha=ScalarParameter() {defaultValue=1, parameterName=IsAttackReady}
   )) {name=Color, nodeColor=<rgba 1 0 0.3047 1>}
  n12 = ScalarParameter() {defaultValue=0.95, parameterName=Radius}
  NamedRerouteDeclaration(None=MaterialFunctionCall(DistanceField=MaterialFunctionCall!(UVs, Radius=@n12) {materialFunction=/FortUILibrary/MaterialFunctions/Shapes/SDFs/MF_SDF_Circle}.SDF) {materialFunction=&MF_UI_AntiAliasing}.Inner) {name=BackingCircle, nodeColor=<rgba 0.0312 1 0 1>}
  NamedRerouteDeclaration(None=Subtract(A=NamedRerouteUsage(), B=NamedRerouteUsage())) {name=BackingCircleCutout, nodeColor=<rgba 0 0.6719 1 1>}
  NamedRerouteUsage()
  NamedRerouteUsage()
  n13 = NamedRerouteUsage()
  n14 = LinearInterpolate(Alpha=Multiply(A=NamedRerouteUsage(), B=ScalarParameter() {parameterName=IsCooldown}))
  n15 = NamedRerouteUsage()
  n16 = ScalarParameter() {parameterName=IsCooldown}
  n17 = LinearInterpolate(
   B=LinearInterpolate(Alpha=Multiply(A=@n15, B=@n16)) {desc=Used to Remove white Outline Resulting other lerps and subtract}
   Alpha=ScalarParameter() {parameterName=IsCooldown}
  )
  n18 = LinearInterpolate(A=NamedRerouteUsage(), Alpha=ScalarParameter() {parameterName=IsCooldown}) {constB=0}
  n19 = TextureSampleParameter2D(UVs=MaterialFunctionCall(Texture Scale=ScalarParameter() {defaultValue=1.15, parameterName=AbilityIconScale}) {materialFunction=&ScaleUVsByCenter}.UVs) {parameterName=MSDFIcon, texture=/BeastGirlStormClawsUI/HUD/Textures/T_UI_MSDF_BentBaton_Ability_Claws}
  n20 = NamedRerouteDeclaration(None=MaterialFunctionCall(
    DistanceField=MaterialFunctionCall!(MSDF RGB=@n19.RGB) {materialFunction=/FortUILibrary/MaterialFunctions/Shapes/Modifiers/MF_UI_MSDFtoSDF}.SDF
    Threshold=Constant() {r=0.22}
   ) {materialFunction=&MF_UI_AntiAliasing}.Outer) {name=AA'dAbilityIcon, nodeColor=<rgba 1 0 0.3516 1>}
  Saturate(None=LinearInterpolate(A=LinearInterpolate(A=@n14, B=@n17, Alpha=@n13) {constA=1, constB=0}, B=Multiply(A=@n18, B=@n20), Alpha=NamedRerouteUsage()) {constB=0})
  n21 = Add(A=TextureCoordinate(), B=Constant2Vector() {g=-0.015, r=-0.01})
  n22 = ScalarParameter() {defaultValue=1.15, parameterName=AbilityIconScale}
  n23 = MaterialFunctionCall(MSDF RGB=TextureSampleParameter2D(UVs=MaterialFunctionCall(UVs=@n21, Texture Scale=@n22) {materialFunction=&ScaleUVsByCenter}.UVs) {parameterName=MSDFIcon, texture=&T_UI_MSDF_BentBaton_Ability_Claws}.RGB) {materialFunction=&MF_UI_MSDFtoSDF}
  n24 = Constant() {r=0.2}
  n25 = Multiply(A=NamedRerouteDeclaration(None=MaterialFunctionCall(DistanceField=@n23.SDF, Threshold=@n24) {materialFunction=&MF_UI_AntiAliasing}.Outer) {name=AbilityIconShadow, nodeColor=<rgba 0 1 0.3672 1>}) {constB=0.5}
  n26 = Multiply(A=NamedRerouteUsage()) {constB=0.25}
  n27 = Multiply(A=Multiply(A=NamedRerouteUsage(), B=NamedRerouteUsage())) {constB=0.9}
  Saturate(None=LinearInterpolate(A=Max(A=@n20, B=@n25), B=Add(A=@n26, B=@n27), Alpha=ScalarParameter() {parameterName=IsCooldown}) {constA=0.5})
 ]
}
