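# Translucent UI material for a bent baton glow in the HUD, built on a radial gradient.
# Center-scaled UVs, sized by RadialGradientSize times MasterScale, feed MF_UI_RadialGradient along with start and end parameters.
# Its Smooth output blends internal and external pink-magenta color parameters and, dimmed by a constant, gives an opacity value.
# Both results end in named reroutes, with no connection to the material outputs shown.
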Material /BeastGirlStormClawsUI/HUD/Materials/M_UI_BentBaton_Glow
graph {
 properties {bAutomaticallySetUsageInEditor=false, blendMode=BLEND_Translucent, materialDomain=MD_UI, usedShadingModels=Default Lit}
 nodes [
  n1 = ScalarParameter() {defaultValue=1, parameterName=RadialGradientSize}
  n2 = ScalarParameter() {defaultValue=1, parameterName=MasterScale}
  n3 = MaterialFunctionCall(
   UV=MaterialFunctionCall(Texture Scale=Multiply(A=@n1, B=@n2)) {materialFunction=&ScaleUVsByCenter}.UVs
   GradientStart=ScalarParameter() {parameterName=GradientStart}
   GradientEnd=ScalarParameter() {defaultValue=1, parameterName=GradientEnd}
  ) {materialFunction=&MF_UI_RadialGradient}
  NamedRerouteDeclaration(None=LinearInterpolate(
    A=VectorParameter() {parameterName=RadialInternalGradientColor, defaultValue=<rgba 0.5333 0.0887 0.3763 1>}.RGB
    B=VectorParameter() {parameterName=RadialExternalGradientColor, defaultValue=<rgba 0.2918 0.0075 0.1912 1>}.RGB
    Alpha=@n3.Smooth
   )) {name=RadialGradientColor, nodeColor=<rgba 1 0.7031 0 1>}
  NamedRerouteDeclaration(None=Multiply(A=@n3.Smooth) {constB=0.6}) {name=GradientOpacity, nodeColor=<rgba 0.3828 0 1 1>}
 ]
}
