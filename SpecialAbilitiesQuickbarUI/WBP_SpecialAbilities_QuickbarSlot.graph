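# Base quickbar slot widget for special abilities that reacts to an ability's active and cooldown gameplay tags on the owning player.
# Listen events store each tag and restart async tag add/remove waits, which drive animation hooks left empty here, as are the keybind and visual-update functions.
# Remaining cooldown comes from the pawn's Fort ability system effects; active-tag removal is only watched when no cooldown is running.
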WidgetBlueprint /SpecialAbilitiesQuickbarUI/WBP_SpecialAbilities_QuickbarSlot
graphs [
 Assign Keybind Action {nodes [Assign Keybind Action!() -> (then, NewBoundAction:Name) {is=title K2Node_FunctionEntry}]}
 EventGraph {nodes [
   n1 = ListenForActiveEvents!() -> (OutputDelegate, then, ActiveGameplayTag:Gameplay Tag Structure) {is=title K2Node_CustomEvent}
   > n2 = Branch(Condition="Not Equal (GameplayTag)"(
     A:Gameplay Tag Structure=@n1.ActiveGameplayTag
     B:Gameplay Tag Structure=Get ActiveGameplayTag() -> (ActiveGameplayTag:Gameplay Tag Structure)
    ) -> (ReturnValue:bool=~) {is=title K2Node_PromotableOperator})
   > n3 = StopAllAnimations(self:User Widget Object Reference)
   EndAction(
    execute=[@n3.then, @n2.else]
    self:Ability Async Object Reference=[
     Get AsyncAddActiveTagAction() -> (AsyncAddActiveTagAction:Ability Async Wait Gameplay Tag Added Object Reference)
     Get AsyncRemoveActiveTagAction() -> (AsyncRemoveActiveTagAction:Ability Async Wait Gameplay Tag Removed Object Reference)
    ]
   )
   > n4 = Set ActiveGameplayTag(ActiveGameplayTag:Gameplay Tag Structure=@n1.ActiveGameplayTag) -> (Output_Get:Gameplay Tag Structure)
   > Branch(Condition=IsGameplayTagValid!(GameplayTag:Gameplay Tag Structure=@n4) -> (ReturnValue:bool))
   > n5 = WaitGameplayTagAddToActor!(
    execute:Exec
    TargetActor:Actor Object Reference=GetOwningPlayer()
    Tag:Gameplay Tag Structure=Get ActiveGameplayTag()
    OnlyTriggerOnce:bool=false
   ) -> (then:Exec, AsyncTaskProxy:Ability Async Wait Gameplay Tag Added Object Reference, Added:Exec) {is=title K2Node_AsyncAction}
   > EndAction(self=Get AsyncAddActiveTagAction())
   > Set AsyncAddActiveTagAction(AsyncAddActiveTagAction:Ability Async Wait Gameplay Tag Added Object Reference=@n5) -> (Output_Get:Ability Async Wait Gameplay Tag Added Object Reference)
   PlayAbilityActiveAnim(execute=@n5.Added)
   n6 = Get CooldownGameplayTag() -> (CooldownGameplayTag:Gameplay Tag Structure)
   > n7 = Branch(Condition=AND Boolean(A=IsGameplayTagValid(GameplayTag=@n6), B=IsCooldownEnabled!() -> (ReturnValue:bool)))
   n8 = WaitGameplayTagRemoveFromActor!(
    execute:Exec=@n7.else
    TargetActor:Actor Object Reference=GetOwningPlayer()
    Tag:Gameplay Tag Structure=Get ActiveGameplayTag()
    OnlyTriggerOnce:bool=false
   ) -> (then:Exec, AsyncTaskProxy:Ability Async Wait Gameplay Tag Removed Object Reference, Removed:Exec) {is=title K2Node_AsyncAction}
   > EndAction(self=Get AsyncRemoveActiveTagAction())
   > Set AsyncRemoveActiveTagAction(AsyncRemoveActiveTagAction:Ability Async Wait Gameplay Tag Removed Object Reference=@n8) -> (Output_Get:Ability Async Wait Gameplay Tag Removed Object Reference)
   PlayAbilityEndedAnim(execute=@n8.Removed)
   n9 = ListenForCooldownEvents!() -> (OutputDelegate, then, CooldownGameplayTag:Gameplay Tag Structure) {is=title K2Node_CustomEvent}
   > n10 = Branch(Condition="Not Equal (GameplayTag)"(A=@n9.CooldownGameplayTag, B=Get CooldownGameplayTag()) -> (ReturnValue=~))
   > n11 = StopAllAnimations()
   EndAction(
    execute=[@n11.then, @n10.else]
    self=[
     Get AsyncRemoveCooldownTagAction() -> (AsyncRemoveCooldownTagAction:Ability Async Wait Gameplay Tag Removed Object Reference)
     Get AsyncAddCooldownTagAction() -> (AsyncAddCooldownTagAction:Ability Async Wait Gameplay Tag Added Object Reference)
    ]
   )
   > n12 = Set CooldownGameplayTag(CooldownGameplayTag:Gameplay Tag Structure=@n9.CooldownGameplayTag) -> (Output_Get:Gameplay Tag Structure)
   > Branch(Condition=IsGameplayTagValid(GameplayTag=@n12))
   > n13 = WaitGameplayTagAddToActor(TargetActor=GetOwningPlayer(), Tag=Get CooldownGameplayTag(), OnlyTriggerOnce=false)
   > EndAction(self=Get AsyncAddCooldownTagAction())
   > Set AsyncAddCooldownTagAction(AsyncAddCooldownTagAction:Ability Async Wait Gameplay Tag Added Object Reference=@n13) -> (Output_Get:Ability Async Wait Gameplay Tag Added Object Reference)
   Branch(execute=@n13.Added, Condition=IsCooldownEnabled())
   > PlayCooldownStartedAnim()
   > n14 = WaitGameplayTagRemoveFromActor(TargetActor=GetOwningPlayer(), Tag=Get CooldownGameplayTag(), OnlyTriggerOnce=false)
   > EndAction(self=Get AsyncRemoveCooldownTagAction())
   > Set AsyncRemoveCooldownTagAction(AsyncRemoveCooldownTagAction:Ability Async Wait Gameplay Tag Removed Object Reference=@n14) -> (Output_Get:Ability Async Wait Gameplay Tag Removed Object Reference)
   PlayCooldownEndedAnim(execute=@n14.Removed)
   n15 = Event PreConstruct()
   > Branch(Condition=@n15.IsDesignTime)
   > Update Visuals Based on Index()
   Event Construct()
   > Update Visuals Based on Index()
  ]}
 GetAbilityDataTableRow {nodes [
   GetAbilityDataTableRow() {is=title K2Node_FunctionEntry}
   > Return Node!(execute, DataTableRowName:Name=Get DataTableRowName() -> (DataTableRowName:Name=None))
  ]}
 GetRemainingCooldownTime {nodes [
   GetRemainingCooldownTime() {is=title K2Node_FunctionEntry}
   > n1 = GetActiveEffectsWithAllTags!(
    execute
    self:Ability System Component Object Reference=GetFortAbilitySystemComponent!(Actor:Actor Object Reference=GetOwningPlayerPawn!(self:User Widget Object Reference) -> (ReturnValue:Pawn Object Reference)) -> (ReturnValue:Fort Ability System Component Object Reference)
    Tags_GameplayTags:Array of Gameplay Tag Structures=Make Array("[0]"=Get CooldownGameplayTag())
   ) -> (then, ReturnValue:Array of Active Gameplay Effect Handle Structures)
   > n2 = Branch(Condition="integer > integer"(A=Length(TargetArray:Array of Active Gameplay Effect Handle Structures=@n1)))
   > n3 = GetActiveGameplayEffectRemainingDuration!(
    execute
    ActiveHandle:Active Gameplay Effect Handle Structure="Get (a ref)"(Array:Array of Active Gameplay Effect Handle Structures=@n1, Dimension 1=0) -> (Output:"Active Gameplay Effect Handle Structure (by ref)")
   ) -> (then, ReturnValue:float)
   > Return Node!(execute, RemainingTime:float=@n3)
   Return Node(execute=@n2.else, RemainingTime=0)
  ]}
 IsCooldownEnabled {nodes [
   IsCooldownEnabled() {is=title K2Node_FunctionEntry}
   > Return Node!(execute, ReturnValue:bool="float > float"(A=GetRemainingCooldownTime!(self) -> (RemainingTime:float) {is=title K2Node_CallFunction}))
  ]}
 PlayAbilityActiveAnim {nodes [PlayAbilityActiveAnim() {is=title K2Node_FunctionEntry}]}
 PlayAbilityEndedAnim {nodes [PlayAbilityEndedAnim() {is=title K2Node_FunctionEntry}]}
 PlayCooldownEndedAnim {nodes [PlayCooldownEndedAnim() {is=title K2Node_FunctionEntry}]}
 PlayCooldownStartedAnim {nodes [PlayCooldownStartedAnim() {is=title K2Node_FunctionEntry}]}
 StoreDataTableRowName {nodes [
   n1 = StoreDataTableRowName!() -> (then, DataTableRowName:Name) {is=title K2Node_FunctionEntry}
   > Set DataTableRowName(DataTableRowName:Name=@n1) -> (Output_Get:Name=None)
   > Return Node!(execute)
  ]}
 Update Visuals Based on Index {nodes [Update Visuals Based on Index() {is=title K2Node_FunctionEntry}]}
]
variables [
 AbilityBoundAction {type=string, default=None}
 AbilityIcon {type=Texture2D}
 ActiveGameplayTag {type=GameplayTag, default {tagName=None}}
 AsyncAddActiveTagAction {type=AbilityAsync_WaitGameplayTagAdded, default=null}
 AsyncAddCooldownTagAction {type=AbilityAsync_WaitGameplayTagAdded, default=null}
 AsyncRemoveActiveTagAction {type=AbilityAsync_WaitGameplayTagRemoved, default=null}
 AsyncRemoveCooldownTagAction {type=AbilityAsync_WaitGameplayTagRemoved, default=null}
 CooldownGameplayTag {type=GameplayTag, default {tagName=None}}
 DataTableRowName {type=string, default=None}
]
widgets {info {inheritedWidgetCount=0, parentClass=&UserWidget, rootWidgetClass=null, widgetCount=0}, widgets []}
