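# HUD quickbar slot for a BeastGirl StormClaws ability, extending WBP_SpecialAbilities_QuickbarSlot with icon, background, glow and keybind.
# Play functions run active, ended and cooldown animations, while a looping timer feeds cooldown progress into the icon material's Fill.
# The slot whose data row is Storm also shows a radial around the keybind, timed by the charge-start gameplay event's magnitude.
# Its timers stop when the AirDive loop tag is removed.
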WidgetBlueprint /BeastGirlStormClawsUI/HUD/Widgets/WBP_BeastGirl_AbilitiesQuickbarSlot
animations [
 Anim_AbilityActive {
  binding_count=4
  length_seconds=0.5
  object=&self:Anim_AbilityActive
  sequence {bindings [
    Image_Icon {
     id=#1
     possessed_class=Image
     tracks [
      {
       type=MovieSceneColorTrack
       display_name=Color and Opacity
       sections [{
         type=MovieSceneColorSection
         end_frame=10
         start_frame=0
         channels [
          Color.R {type=MovieSceneScriptingFloatChannel, keys [[0,1],[5,0.4],[10,0.4]]}
          Color.G {type=MovieSceneScriptingFloatChannel, keys [[0,1],[5,0.4],[10,0.4]]}
          Color.B {type=MovieSceneScriptingFloatChannel, keys [[0,1],[5,0.4],[10,0.4]]}
          Color.A {type=MovieSceneScriptingFloatChannel, const [1,3]}
         ]
        }]
      }
      {
       type=MovieScene2DTransformTrack
       sections [{
         type=MovieScene2DTransformSection
         end_frame=10
         start_frame=0
         channels [
          Scale.X {type=MovieSceneScriptingFloatChannel, keys [[0,0.7805],[5,0.58],[10,0.58]]}
          Scale.Y {type=MovieSceneScriptingFloatChannel, keys [[0,0.7805],[5,0.58],[10,0.58]]}
         ]
        }]
      }
     ]
    }
    Image_Background {
     id=#2
     possessed_class=Image
     tracks [
      {
       type=MovieScene2DTransformTrack
       sections [{
         type=MovieScene2DTransformSection
         end_frame=10
         start_frame=0
         channels [
          Scale.X {type=MovieSceneScriptingFloatChannel, keys [[0,1],[5,0.9]]}
          Scale.Y {type=MovieSceneScriptingFloatChannel, keys [[0,1],[5,0.9]]}
         ]
        }]
      }
      {
       type=MovieSceneFloatTrack
       display_name=Render Opacity
       sections [{
         type=MovieSceneFloatSection
         end_frame=10
         start_frame=0
         channels [None {type=MovieSceneScriptingFloatChannel, keys [[0,0.8],[5,0.3],[10,0.3]]}]
        }]
      }
     ]
    }
    Image_Background_Overlay {
     id=#3
     possessed_class=Image
     tracks [
      {
       type=MovieScene2DTransformTrack
       sections [{
         type=MovieScene2DTransformSection
         end_frame=10
         start_frame=0
         channels [
          Scale.X {type=MovieSceneScriptingFloatChannel, keys [[0,0.7805],[5,0.6805]]}
          Scale.Y {type=MovieSceneScriptingFloatChannel, keys [[0,0.7805],[5,0.6805]]}
         ]
        }]
      }
      {
       type=MovieSceneColorTrack
       display_name=Color and Opacity
       sections [{
         type=MovieSceneColorSection
         end_frame=10
         start_frame=0
         channels [
          Color.R {type=MovieSceneScriptingFloatChannel, keys [[0,1],[5,0.6]]}
          Color.G {type=MovieSceneScriptingFloatChannel, keys [[0,1],[5,0.6]]}
          Color.B {type=MovieSceneScriptingFloatChannel, keys [[0,1],[5,0.6]]}
          Color.A {type=MovieSceneScriptingFloatChannel, const [1,1]}
         ]
        }]
      }
     ]
    }
    Image_Glow {
     id=#4
     possessed_class=Image
     tracks [{
       type=MovieSceneFloatTrack
       display_name=Render Opacity
       sections [{type=MovieSceneFloatSection, end_frame=10, start_frame=0, channels [None {type=MovieSceneScriptingFloatChannel, keys [[0,1],[5,0.5]]}]}]
      }]
    }
   ]}
 }
 Anim_AbilityEnded {
  binding_count=4
  length_seconds=0.1
  object=&self:Anim_AbilityEnded
  sequence {bindings [
    Image_Glow {
     id=#5
     possessed_class=Image
     tracks [{
       type=MovieSceneFloatTrack
       display_name=Render Opacity
       sections [{type=MovieSceneFloatSection, end_frame=2, start_frame=0, channels [None {type=MovieSceneScriptingFloatChannel, keys [[0,0.5],[1,1]]}]}]
      }]
    }
    Image_Background {
     id=#6
     possessed_class=Image
     tracks [
      {
       type=MovieScene2DTransformTrack
       sections [{
         type=MovieScene2DTransformSection
         end_frame=2
         start_frame=0
         channels [
          Scale.X {type=MovieSceneScriptingFloatChannel, keys [[0,0.9],[1,1]]}
          Scale.Y {type=MovieSceneScriptingFloatChannel, keys [[0,0.9],[1,1]]}
         ]
        }]
      }
      {
       type=MovieSceneFloatTrack
       display_name=Render Opacity
       sections [{
         type=MovieSceneFloatSection
         end_frame=2
         start_frame=0
         channels [None {type=MovieSceneScriptingFloatChannel, keys [[0,0.3],[1,0.8],[2,0.8]]}]
        }]
      }
     ]
    }
    Image_Background_Overlay {
     id=#7
     possessed_class=Image
     tracks [
      {
       type=MovieScene2DTransformTrack
       sections [{
         type=MovieScene2DTransformSection
         end_frame=2
         start_frame=0
         channels [
          Scale.X {type=MovieSceneScriptingFloatChannel, keys [[0,0.6805],[1,0.7805]]}
          Scale.Y {type=MovieSceneScriptingFloatChannel, keys [[0,0.6805],[1,0.7805]]}
         ]
        }]
      }
      {
       type=MovieSceneColorTrack
       display_name=Color and Opacity
       sections [{
         type=MovieSceneColorSection
         end_frame=2
         start_frame=0
         channels [
          Color.R {type=MovieSceneScriptingFloatChannel, keys [[0,0.6],[1,1]]}
          Color.G {type=MovieSceneScriptingFloatChannel, keys [[0,0.6],[1,1]]}
          Color.B {type=MovieSceneScriptingFloatChannel, keys [[0,0.6],[1,1]]}
          Color.A {type=MovieSceneScriptingFloatChannel, const [1,1]}
         ]
        }]
      }
     ]
    }
    Image_Icon {
     id=#8
     possessed_class=Image
     tracks [
      {
       type=MovieSceneColorTrack
       display_name=Color and Opacity
       sections [{
         type=MovieSceneColorSection
         end_frame=2
         start_frame=0
         channels [
          Color.R {type=MovieSceneScriptingFloatChannel, keys [[0,0.4],[1,1]]}
          Color.G {type=MovieSceneScriptingFloatChannel, keys [[0,0.4],[1,1]]}
          Color.B {type=MovieSceneScriptingFloatChannel, keys [[0,0.4],[1,1]]}
          Color.A {type=MovieSceneScriptingFloatChannel, const [1,1]}
         ]
        }]
      }
      {
       type=MovieScene2DTransformTrack
       sections [{
         type=MovieScene2DTransformSection
         end_frame=2
         start_frame=0
         channels [
          Scale.X {type=MovieSceneScriptingFloatChannel, keys [[0,0.58],[1,0.7805],[10,1]]}
          Scale.Y {type=MovieSceneScriptingFloatChannel, keys [[0,0.58],[1,0.7805],[10,1]]}
         ]
        }]
      }
     ]
    }
   ]}
 }
 Anim_CooldownStarted {
  binding_count=4
  length_seconds=0.5
  object=&self:Anim_CooldownStarted
  sequence {bindings [
    Image_Icon {
     id=#9
     possessed_class=Image
     tracks [{
       type=MovieSceneWidgetMaterialTrack
       sections [{
         type=MovieSceneParameterSection
         end_frame=10
         start_frame=0
         channels [
          IsAttackReady {type=MovieSceneScriptingFloatChannel, keys [[0,1],[9.9993,0]]}
          IsCooldown {type=MovieSceneScriptingFloatChannel, keys [[0,0],[10,1]]}
          AbilityIconScale {type=MovieSceneScriptingFloatChannel, keys [[0,1.1],[9.9993,0.9]]}
         ]
        }]
      }]
    }
    Image_Background {
     id=#10
     possessed_class=Image
     tracks [{
       type=MovieSceneWidgetMaterialTrack
       sections [{
         type=MovieSceneParameterSection
         end_frame=10
         start_frame=0
         channels [AbilityReady {type=MovieSceneScriptingFloatChannel, keys [[0,1],[9.9993,0]]}]
        }]
      }]
    }
    Image_Glow {
     id=#11
     possessed_class=Image
     tracks [{
       type=MovieSceneFloatTrack
       display_name=Render Opacity
       sections [{type=MovieSceneFloatSection, end_frame=10, start_frame=0, channels [None {type=MovieSceneScriptingFloatChannel, keys [[0,1],[10,0]]}]}]
      }]
    }
    Image_Background_Overlay {
     id=#12
     possessed_class=Image
     tracks [{
       type=MovieSceneWidgetMaterialTrack
       sections [{
         type=MovieSceneParameterSection
         end_frame=10
         start_frame=0
         channels [
          RadialExternalGradientColorR {type=MovieSceneScriptingFloatChannel, keys [[0,0.1927],[9.9993,0]]}
          RadialExternalGradientColorG {type=MovieSceneScriptingFloatChannel, const [0,2]}
          RadialExternalGradientColorB {type=MovieSceneScriptingFloatChannel, keys [[0,0.1003],[9.9993,0]]}
          RadialExternalGradientColorA {type=MovieSceneScriptingFloatChannel, const [1,2]}
          RadialInternalGradientColorR {type=MovieSceneScriptingFloatChannel, const [0,1]}
          RadialInternalGradientColorG {type=MovieSceneScriptingFloatChannel, const [0,1]}
          RadialInternalGradientColorB {type=MovieSceneScriptingFloatChannel, const [0,1]}
          RadialInternalGradientColorA {type=MovieSceneScriptingFloatChannel, const [1,1]}
         ]
        }]
      }]
    }
   ]}
 }
 Anim_CooldownEnded {
  binding_count=5
  length_seconds=0.5
  object=&self:Anim_CooldownEnded
  sequence {bindings [
    Image_Icon {
     id=#13
     possessed_class=Image
     tracks [{
       type=MovieSceneWidgetMaterialTrack
       sections [{
         type=MovieSceneParameterSection
         end_frame=10
         start_frame=0
         channels [
          IsAttackReady {type=MovieSceneScriptingFloatChannel, keys [[0,0],[9.9993,1]]}
          IsCooldown {type=MovieSceneScriptingFloatChannel, keys [[0,1],[9.9993,0]]}
          AbilityIconScale {type=MovieSceneScriptingFloatChannel, keys [[0,0.9],[9.9993,1.1]]}
         ]
        }]
      }]
    }
    Image_Background_Overlay {
     id=#14
     possessed_class=Image
     tracks [{
       type=MovieSceneWidgetMaterialTrack
       sections [{
         type=MovieSceneParameterSection
         end_frame=10
         start_frame=0
         channels [
          RadialExternalGradientColorR {type=MovieSceneScriptingFloatChannel, keys [[0,0],[9.9993,0.1927]]}
          RadialExternalGradientColorG {type=MovieSceneScriptingFloatChannel, const [0,2]}
          RadialExternalGradientColorB {type=MovieSceneScriptingFloatChannel, keys [[0,0],[9.9993,0.1003]]}
          RadialExternalGradientColorA {type=MovieSceneScriptingFloatChannel, const [1,1]}
          RadialInternalGradientColorR {type=MovieSceneScriptingFloatChannel, const [0,1]}
          RadialInternalGradientColorG {type=MovieSceneScriptingFloatChannel, const [0,1]}
          RadialInternalGradientColorB {type=MovieSceneScriptingFloatChannel, const [0,1]}
          RadialInternalGradientColorA {type=MovieSceneScriptingFloatChannel, const [1,1]}
         ]
        }]
      }]
    }
    Overlay_0 {
     id=#15
     possessed_class=Overlay
     tracks [{
       type=MovieSceneFloatTrack
       display_name=Render Opacity
       sections [{
         type=MovieSceneFloatSection
         end_frame=9
         start_frame=0
         channels [None {type=MovieSceneScriptingFloatChannel, keys [[0,0.25],[2.7,0.25],[7.2,1],[9,1]]}]
        }]
      }]
    }
    Image_Glow {
     id=#16
     possessed_class=Image
     tracks [{
       type=MovieSceneColorTrack
       display_name=Color and Opacity
       sections [{
         type=MovieSceneColorSection
         end_frame=10
         start_frame=0
         channels [
          Color.R {type=MovieSceneScriptingFloatChannel, const [1,1]}
          Color.G {type=MovieSceneScriptingFloatChannel, const [1,1]}
          Color.B {type=MovieSceneScriptingFloatChannel, const [1,1]}
          Color.A {type=MovieSceneScriptingFloatChannel, keys [[0,0],[10,1]]}
         ]
        }]
      }]
    }
    Image_Background {
     id=#17
     possessed_class=Image
     tracks [{
       type=MovieSceneWidgetMaterialTrack
       sections [{
         type=MovieSceneParameterSection
         end_frame=10
         start_frame=0
         channels [AbilityReady {type=MovieSceneScriptingFloatChannel, keys [[0,0],[9.9993,1]]}]
        }]
      }]
    }
   ]}
 }
 Anim_InputRadial {
  binding_count=1
  length_seconds=1
  object=&self:Anim_InputRadial
  sequence {bindings [Image_InputRadial {
     id=#18
     possessed_class=Image
     tracks [{
       type=MovieSceneWidgetMaterialTrack
       sections [{
         type=MovieSceneParameterSection
         end_frame=20
         start_frame=0
         channels [Amount_0-1 {type=MovieSceneScriptingFloatChannel, keys [[0,0],[20,1]]}]
        }]
      }]
    }]}
 }
 Anim_AbilityEnded_Storm {
  binding_count=5
  length_seconds=0.1
  object=&self:Anim_AbilityEnded_Storm
  sequence {bindings [
    Overlay_0 {
     id=#19
     possessed_class=Overlay
     tracks [{
       type=MovieSceneFloatTrack
       display_name=Render Opacity
       sections [{type=MovieSceneFloatSection, end_frame=2, start_frame=0, channels [None {type=MovieSceneScriptingFloatChannel, keys [[0,1],[2,0.25]]}]}]
      }]
    }
    Image_Glow {
     id=#5
     possessed_class=Image
     tracks [{
       type=MovieSceneFloatTrack
       display_name=Render Opacity
       sections [{type=MovieSceneFloatSection, end_frame=2, start_frame=0, channels [None {type=MovieSceneScriptingFloatChannel, keys [[0,0.5],[1,1]]}]}]
      }]
    }
    Image_Background {
     id=#6
     possessed_class=Image
     tracks [
      {
       type=MovieScene2DTransformTrack
       sections [{
         type=MovieScene2DTransformSection
         end_frame=2
         start_frame=0
         channels [
          Scale.X {type=MovieSceneScriptingFloatChannel, keys [[0,0.9],[1,1]]}
          Scale.Y {type=MovieSceneScriptingFloatChannel, keys [[0,0.9],[1,1]]}
         ]
        }]
      }
      {
       type=MovieSceneFloatTrack
       display_name=Render Opacity
       sections [{
         type=MovieSceneFloatSection
         end_frame=2
         start_frame=0
         channels [None {type=MovieSceneScriptingFloatChannel, keys [[0,0.3],[1,0.8],[2,0.8]]}]
        }]
      }
     ]
    }
    Image_Background_Overlay {
     id=#7
     possessed_class=Image
     tracks [
      {
       type=MovieScene2DTransformTrack
       sections [{
         type=MovieScene2DTransformSection
         end_frame=2
         start_frame=0
         channels [
          Scale.X {type=MovieSceneScriptingFloatChannel, keys [[0,0.6805],[1,0.7805]]}
          Scale.Y {type=MovieSceneScriptingFloatChannel, keys [[0,0.6805],[1,0.7805]]}
         ]
        }]
      }
      {
       type=MovieSceneColorTrack
       display_name=Color and Opacity
       sections [{
         type=MovieSceneColorSection
         end_frame=2
         start_frame=0
         channels [
          Color.R {type=MovieSceneScriptingFloatChannel, keys [[0,0.6],[1,1]]}
          Color.G {type=MovieSceneScriptingFloatChannel, keys [[0,0.6],[1,1]]}
          Color.B {type=MovieSceneScriptingFloatChannel, keys [[0,0.6],[1,1]]}
          Color.A {type=MovieSceneScriptingFloatChannel, const [1,1]}
         ]
        }]
      }
     ]
    }
    Image_Icon {
     id=#8
     possessed_class=Image
     tracks [
      {
       type=MovieSceneColorTrack
       display_name=Color and Opacity
       sections [{
         type=MovieSceneColorSection
         end_frame=2
         start_frame=0
         channels [
          Color.R {type=MovieSceneScriptingFloatChannel, keys [[0,0.4],[1,1]]}
          Color.G {type=MovieSceneScriptingFloatChannel, keys [[0,0.4],[1,1]]}
          Color.B {type=MovieSceneScriptingFloatChannel, keys [[0,0.4],[1,1]]}
          Color.A {type=MovieSceneScriptingFloatChannel, const [1,1]}
         ]
        }]
      }
      {
       type=MovieScene2DTransformTrack
       sections [{
         type=MovieScene2DTransformSection
         end_frame=2
         start_frame=0
         channels [
          Scale.X {type=MovieSceneScriptingFloatChannel, keys [[0,0.58],[1,0.7805],[10,1]]}
          Scale.Y {type=MovieSceneScriptingFloatChannel, keys [[0,0.58],[1,0.7805],[10,1]]}
         ]
        }]
      }
     ]
    }
   ]}
 }
]
graphs [
 Assign Keybind Action {nodes [
   n1 = Assign Keybind Action!() -> (then, NewBoundAction:Name=None) {is=title K2Node_FunctionEntry}
   > SetBoundAction!(execute, self:Fort Keybind Widget Object Reference=Get Keybind() -> (Keybind:Keybind Widget Object Reference), NewBoundAction:Name=@n1) -> (then)
   n2 = IsStormAbilitySlot!() -> (IsStormAbilitySlotBool:bool)
   > SetVisibility(
    self=Get Image_InputRadial() -> (Image_InputRadial:Image Object Reference)
    InVisibility=Select(Option 0=Collapsed, Option 1=HitTestInvisible, Index=@n2)
   )
   > Branch(Condition=@n2)
   > ListenForOnChargeInitialized()
   > ListenForLeapRelease()
  ]}
 EventGraph {nodes [
   HandleCooldownFill() {is=title K2Node_CustomEvent}
   n1 = Get CooldownGameplayTag() -> (CooldownGameplayTag:Gameplay Tag Structure)
   > Branch(Condition=IsGameplayTagValid!(GameplayTag:Gameplay Tag Structure=@n1) -> (ReturnValue:bool))
   > n2 = GetCooldownTimeRemainingAndDuration!(
    execute
    Actor:Actor Object Reference=Get PlayerController() -> (PlayerController:Player Controller Object Reference)
    CooldownTags:Gameplay Tag Container Structure=MakeGameplayTagContainerFromTag!(SingleTag:Gameplay Tag Structure=@n1) -> (ReturnValue:Gameplay Tag Container Structure)
   ) -> (then, TimeRemaining:float, CooldownDuration:float)
   n3 = HandleCooldownDuration() {is=title K2Node_CustomEvent}
   n4 = Set Timer by Event(execute=@n2.then, Delegate=@n3, Time=@n2.CooldownDuration, bLooping=false, bMaxOncePerFrame=true)
   > n5 = Set CooldownDurationTimerHandle(CooldownDurationTimerHandle:Timer Handle Structure=@n4) -> (Output_Get:Timer Handle Structure)
   n6 = HandleAbilityIconFIll() {is=title K2Node_CustomEvent}
   n7 = Set Timer by Event(execute=@n5.then, Delegate=@n6, Time=0.05, bLooping=true)
   > Set FillAbilityIconTimerHandle(FillAbilityIconTimerHandle:Timer Handle Structure=@n7) -> (Output_Get:Timer Handle Structure)
   ListenForOnChargeInitialized() {is=title K2Node_CustomEvent}
   > n8 = WaitGameplayEventToActor!(
    execute:Exec
    TargetActor:Actor Object Reference=Get PlayerController()
    EventTag:Gameplay Tag Structure=Get InitialChargeTag() -> (InitialChargeTag:Gameplay Tag Structure)
    OnlyTriggerOnce:bool=false
    OnlyMatchExact:bool=true
   ) -> (then:Exec, AsyncTaskProxy:Ability Async Wait Gameplay Event Object Reference, EventReceived:Exec, Payload:Gameplay Event Data Structure) {is=title K2Node_AsyncAction}
   n9 = Set InputRadialFillDuration(
    execute=@n8.EventReceived
    InputRadialFillDuration:float=Break Gameplay Event Data!(GameplayEventData:"Gameplay Event Data Structure (by ref)"=@n8.Payload) -> (EventMagnitude:float=~, InstancedEventData:Instanced Struct Structure) {is=title K2Node_BreakStruct}
   ) -> (Output_Get:float)
   > n10 = Set Timer by Event(Delegate=Create Event(), Time=@n9, bLooping=false)
   > Set ChargeInputRadialDurationTimerHandle(ChargeInputRadialDurationTimerHandle:Timer Handle Structure=@n10) -> (Output_Get:Timer Handle Structure)
   > n11 = Set Timer by Event(Delegate=Create Event(), Time=Get InputRadialFillRate() -> (InputRadialFillRate:float), bLooping=true)
   > Set ChargeInputRadialFillTimerHandle(ChargeInputRadialFillTimerHandle:Timer Handle Structure=@n11) -> (Output_Get:Timer Handle Structure)
   ListenForLeapRelease() {is=title K2Node_CustomEvent}
   > n12 = WaitGameplayTagRemoveFromActor!(
    execute:Exec
    TargetActor:Actor Object Reference=Get PlayerController()
    Tag:Gameplay Tag Structure=Get AirLeapLoopTag() -> (AirLeapLoopTag:Gameplay Tag Structure)
    OnlyTriggerOnce:bool=false
   ) -> (then:Exec, AsyncTaskProxy:Ability Async Wait Gameplay Tag Removed Object Reference, Removed:Exec) {is=title K2Node_AsyncAction}
   InvalidateRadialInputFillTimers(execute=@n12.Removed)
   n13 = Event PreConstruct()
   > Parent: PreConstruct(IsDesignTime=@n13.IsDesignTime)
   > n14 = GetDynamicMaterial(self=Get Image_Background() -> (Image_Background:Image Object Reference))
   > n15 = Set MIDBackground(MIDBackground:Material Instance Dynamic Object Reference=@n14) -> (Output_Get:Material Instance Dynamic Object Reference)
   > SetScalarParameterValue(
    self=@n15
    ParameterName=RotationSpeed
    Value=Select(Option 0:float=0.5, Option 1:float=-0.6, Index:int=Get RotationAnimSeed() -> (RotationAnimSeed:int=0)) -> (ReturnValue:float)
   )
   > SetScalarParameterValue(self=Get MIDBackground() -> (MIDBackground:Material Instance Dynamic Object Reference), ParameterName=IsMaterialTime, Value=1)
   Event Construct()
   > Parent: Construct()
   > Set PlayerController(PlayerController:Player Controller Object Reference=GetOwningPlayer()) -> (Output_Get:Player Controller Object Reference)
   > n16 = GetDynamicMaterial(self=Get Image_Background())
   > n17 = Set MIDBackground(MIDBackground=@n16)
   > SetScalarParameterValue(self=@n17, ParameterName=RotationSpeed, Value=Select(Option 0=0.5, Option 1=-0.6, Index=Get RotationAnimSeed() -> (RotationAnimSeed=0)))
   > n18 = GetDynamicMaterial(self=Get Image_Icon() -> (Image_Icon:Image Object Reference))
   > Set MIDAbilityIcon(MIDAbilityIcon:Material Instance Dynamic Object Reference=@n18) -> (Output_Get:Material Instance Dynamic Object Reference)
   > n19 = GetDynamicMaterial(self=Get Image_InputRadial())
   > n20 = Set MID_InputRadial(MID_InputRadial:Material Instance Dynamic Object Reference=@n19) -> (Output_Get:Material Instance Dynamic Object Reference)
   > SetScalarParameterValue(self=@n20, ParameterName=Amount_0-1, Value=0)
   Clear and Invalidate Timer by Handle(execute=@n22.then, Handle=Get FillAbilityIconTimerHandle() -> (FillAbilityIconTimerHandle:Timer Handle Structure))
   n21 = GetCooldownTimeRemainingAndDuration(execute=@n6.then, Actor=Get PlayerController(), CooldownTags=MakeGameplayTagContainerFromTag(SingleTag=Get CooldownGameplayTag()))
   > SetScalarParameterValue(
    self=Get MIDAbilityIcon() -> (MIDAbilityIcon:Material Instance Dynamic Object Reference)
    ParameterName=Fill
    Value=float - float(A=1, B=float / float(A=@n21, B=@n21.CooldownDuration))
   )
   Event Destruct()
   n22 = Clear and Invalidate Timer by Handle(execute=@n3.then, Handle=Get CooldownDurationTimerHandle() -> (CooldownDurationTimerHandle:Timer Handle Structure))
  ]}
 FillRadialInput {nodes [
   FillRadialInput() {is=title K2Node_FunctionEntry}
   n1 = Get InputRadialFillDuration() -> (InputRadialFillDuration:float)
   > n2 = Branch(Condition="float >= float"(A=Get InputRadialFillElapsedTime() -> (InputRadialFillElapsedTime:float), B=float - float(A=@n1, B=0.05)))
   > SetScalarParameterValue(self=Get MID_InputRadial() -> (MID_InputRadial:Material Instance Dynamic Object Reference), ParameterName=Amount_0-1, Value=1)
   > InvalidateRadialInputFillTimers()
   SetScalarParameterValue(
    execute=@n2.else
    self=Get MID_InputRadial()
    ParameterName=Amount_0-1
    Value=float / float(A=Get InputRadialFillElapsedTime(), B=Get InputRadialFillDuration())
   )
   > Set InputRadialFillElapsedTime(InputRadialFillElapsedTime:float=float + float(A=Get InputRadialFillElapsedTime(), B=Get InputRadialFillRate())) -> (Output_Get:float)
  ]}
 InvalidateRadialInputFillTimers {nodes [
   InvalidateRadialInputFillTimers() {is=title K2Node_FunctionEntry}
   > Clear and Invalidate Timer by Handle(Handle=Get ChargeInputRadialDurationTimerHandle() -> (ChargeInputRadialDurationTimerHandle:Timer Handle Structure))
   > Clear and Invalidate Timer by Handle(Handle=Get ChargeInputRadialFillTimerHandle() -> (ChargeInputRadialFillTimerHandle:Timer Handle Structure))
   > Set InputRadialFillElapsedTime(InputRadialFillElapsedTime=0)
  ]}
 IsStormAbilitySlot {nodes [
   IsStormAbilitySlot() {is=title K2Node_FunctionEntry}
   > Return Node!(
    execute
    IsStormAbilitySlotBool:bool="Equal (Name)"(A=GetAbilityDataTableRow!() -> (DataTableRowName:Name=None), B=Get StormAttackName() -> (StormAttackName:Name=None))
   )
  ]}
 PlayAbilityActiveAnim {nodes [
   PlayAbilityActiveAnim() {is=title K2Node_FunctionEntry}
   > PlayAnimation(InAnimation=Get Anim_AbilityActive() -> (Anim_AbilityActive:Widget Animation Object Reference))
  ]}
 PlayAbilityEndedAnim {nodes [
   PlayAbilityEndedAnim() {is=title K2Node_FunctionEntry}
   > n1 = Branch(Condition=IsStormAbilitySlot() {is=title K2Node_CallFunction})
   > PlayAnimationForward(InAnimation=Get Anim_AbilityEnded_Storm() -> (Anim_AbilityEnded_Storm:Widget Animation Object Reference))
   > ResetInputRadiaFilll()
   > InvalidateRadialInputFillTimers() {is=title K2Node_CallFunction}
   PlayAnimationForward(execute=@n1.else, InAnimation=Get Anim_AbilityEnded() -> (Anim_AbilityEnded:Widget Animation Object Reference))
  ]}
 PlayCooldownEndedAnim {nodes [
   PlayCooldownEndedAnim() {is=title K2Node_FunctionEntry}
   > PlayAnimation(InAnimation=Get Anim_CooldownEnded() -> (Anim_CooldownEnded:Widget Animation Object Reference))
   > Clear and Invalidate Timer by Handle(Handle=Get CooldownDurationTimerHandle())
   > Clear and Invalidate Timer by Handle(Handle=Get FillAbilityIconTimerHandle())
  ]}
 PlayCooldownStartedAnim {nodes [
   PlayCooldownStartedAnim() {is=title K2Node_FunctionEntry}
   > PlayAnimation(InAnimation=Get Anim_CooldownStarted() -> (Anim_CooldownStarted:Widget Animation Object Reference))
   > HandleCooldownFill() {is=title K2Node_CallFunction}
  ]}
 ResetInputRadiaFilll {nodes [
   ResetInputRadiaFilll() {is=title K2Node_FunctionEntry}
   > SetScalarParameterValue(self=Get MID_InputRadial(), ParameterName=Amount_0-1, Value=0)
  ]}
 Update Visuals Based on Index {nodes [
   Update Visuals Based on Index() {is=title K2Node_FunctionEntry}
   > SetTextureParameterValue(self=Get MIDAbilityIcon(), ParameterName=MSDFIcon, Value=Get AbilityIcon() -> (AbilityIcon:Texture 2D Object Reference))
  ]}
]
variables [
 AirLeapLoopTag {type=GameplayTag, default {tagName=GameplayCue.Abilities.BeastGirl.StormClaws.AirDive.Loop}}
 ChargeInputRadialDurationTimerHandle {type=TimerHandle}
 ChargeInputRadialFillTimerHandle {type=TimerHandle}
 CooldownDurationTimerHandle {type=TimerHandle}
 FillAbilityIconTimerHandle {type=TimerHandle}
 InitialChargeTag {type=GameplayTag, default {tagName=Ability.BeastGirl.StormClaws.StormSlam.InitialChargeStarted}}
 InputRadialFillDuration {type=number}
 InputRadialFillElapsedTime {type=number, default=0.3}
 InputRadialFillRate {type=number, default=0.03}
 MID_InputRadial {type=MaterialInstanceDynamic}
 MIDAbilityIcon {type=MaterialInstanceDynamic}
 MIDBackground {type=MaterialInstanceDynamic}
 PlayerController {type=PlayerController, default=null}
 RotationAnimSeed {type=integer}
 StormAttackName {type=string, default=Storm}
]
widgets {
 info {
  inheritedWidgetCount=0
  parentClass=/SpecialAbilitiesQuickbarUI/WBP_SpecialAbilities_QuickbarSlot.WBP_SpecialAbilities_QuickbarSlot_C
  rootWidgetClass=&SizeBox
  widgetCount=11
 }
 widgets [
  SizeBox_QuickbarSize {
   widgetClassPath=&SizeBox
   widget=SizeBox_QuickbarSize
   properties {bOverride_HeightOverride=true, bOverride_WidthOverride=true, heightOverride=120, widthOverride=120}
  }
  Overlay_5 {widgetClassPath=&Overlay, parent=SizeBox_QuickbarSize, widget=Overlay_5, slot {class=SizeBoxSlot}}
  Overlay_2 {
   widgetClassPath=&Overlay
   parent=Overlay_5
   widget=Overlay_2
   slot {class=OverlaySlot, properties {horizontalAlignment=HAlign_Fill, verticalAlignment=VAlign_Fill}}
  }
  Image_Background {
   widgetClassPath=&Image
   bIsVariable=true
   parent=Overlay_2
   widget=Image_Background
   properties {
    renderOpacity=0.8
    brush {
     bIsDynamicallyLoaded=false
     drawAs=Image
     imageType=FullColor
     mirroring=NoMirror
     resourceName=None
     resourceObject=/BeastGirlStormClawsUI/HUD/Materials/M_UI_BentBaton_Ability_Background
     tiling=NoTile
     imageSize=<164 164>
     margin=<ltrb 0 0 0 0>
     outlineSettings {
      bUseBrushTransparency=false
      roundingType=HalfHeightRadius
      width=0
      color {colorUseRule=UseColor_Specified, specifiedColor=<rgba 0 0 0 0>}
      cornerRadii=<0 0 0 0>
     }
     tintColor {colorUseRule=UseColor_Specified, specifiedColor=<rgba 1 1 1 1>}
     uVRegion {bIsValid=false, max=<0 0>, min=<0 0>}
    }
   }
   slot {class=OverlaySlot, properties {horizontalAlignment=HAlign_Fill, verticalAlignment=VAlign_Fill}}
  }
  Image_Background_Overlay {
   widgetClassPath=&Image
   bIsVariable=true
   parent=Overlay_2
   widget=Image_Background_Overlay
   properties {
    brush {
     bIsDynamicallyLoaded=false
     drawAs=Image
     imageType=FullColor
     mirroring=NoMirror
     resourceName=None
     resourceObject=/BeastGirlStormClawsUI/HUD/Materials/MI_UI_BentBaton_Circle
     tiling=NoTile
     imageSize=<128 128>
     margin=<ltrb 0 0 0 0>
     outlineSettings {
      bUseBrushTransparency=false
      roundingType=HalfHeightRadius
      width=0
      color {colorUseRule=UseColor_Specified, specifiedColor=<rgba 0 0 0 0>}
      cornerRadii=<0 0 0 0>
     }
     tintColor {colorUseRule=UseColor_Specified, specifiedColor=<rgba 1 1 1 0.75>}
     uVRegion {bIsValid=false, max=<0 0>, min=<0 0>}
    }
    renderTransform {angle=0, scale=<0.7805 0.7805>, shear=<0 0>, translation=<0 0>}
   }
   slot {class=OverlaySlot, properties {horizontalAlignment=HAlign_Fill, verticalAlignment=VAlign_Fill}}
  }
  Image_Glow {
   widgetClassPath=&Image
   bIsVariable=true
   parent=Overlay_2
   widget=Image_Glow
   properties {
    brush {
     bIsDynamicallyLoaded=false
     drawAs=Image
     imageType=FullColor
     mirroring=NoMirror
     resourceName=None
     resourceObject=/BeastGirlStormClawsUI/HUD/Materials/M_UI_BentBaton_Glow
     tiling=NoTile
     imageSize=<164 164>
     margin=<ltrb 0 0 0 0>
     outlineSettings {
      bUseBrushTransparency=false
      roundingType=HalfHeightRadius
      width=0
      color {colorUseRule=UseColor_Specified, specifiedColor=<rgba 0 0 0 0>}
      cornerRadii=<0 0 0 0>
     }
     tintColor {colorUseRule=UseColor_Specified, specifiedColor=<rgba 1 1 1 0.75>}
     uVRegion {bIsValid=false, max=<0 0>, min=<0 0>}
    }
    renderTransform {angle=0, scale=<1.5 1.5>, shear=<0 0>, translation=<0 0>}
   }
   slot {class=OverlaySlot, properties {horizontalAlignment=HAlign_Fill, verticalAlignment=VAlign_Fill}}
  }
  Image_Icon {
   widgetClassPath=&Image
   bIsVariable=true
   parent=Overlay_2
   widget=Image_Icon
   properties {
    brush {
     bIsDynamicallyLoaded=false
     drawAs=Image
     imageType=FullColor
     mirroring=NoMirror
     resourceName=None
     resourceObject=/BeastGirlStormClawsUI/HUD/Materials/M_UI_BentBaton_Ability_Icon
     tiling=NoTile
     imageSize=<128 128>
     margin=<ltrb 0 0 0 0>
     outlineSettings {
      bUseBrushTransparency=false
      roundingType=FixedRadius
      width=1.9847
      color {colorUseRule=UseColor_Specified, specifiedColor=<rgba 0.1146 0.1146 0.1146 0>}
      cornerRadii=<0 0 0 0>
     }
     tintColor {colorUseRule=UseColor_Specified, specifiedColor=<rgba 1 1 1 1>}
     uVRegion {bIsValid=false, max=<0 0>, min=<0 0>}
    }
    renderTransform {angle=0, scale=<0.7805 0.7805>, shear=<0 0>, translation=<0 0>}
   }
   slot {class=OverlaySlot, properties {horizontalAlignment=HAlign_Fill, verticalAlignment=VAlign_Fill}}
  }
  SizeBox {
   widgetClassPath=&SizeBox
   parent=Overlay_5
   widget=SizeBox
   properties {heightOverride=41, minDesiredHeight=45, widthOverride=41}
   slot {class=OverlaySlot, properties {horizontalAlignment=HAlign_Right, verticalAlignment=VAlign_Bottom}}
  }
  Overlay_0 {
   widgetClassPath=&Overlay
   parent=SizeBox
   widget=Overlay_0
   slot {class=SizeBoxSlot, properties {horizontalAlignment=HAlign_Center, verticalAlignment=VAlign_Center}}
  }
  Image_InputRadial {
   widgetClassPath=&Image
   bIsVariable=true
   parent=Overlay_0
   widget=Image_InputRadial
   properties {brush {
     bIsDynamicallyLoaded=false
     drawAs=Image
     imageType=FullColor
     mirroring=NoMirror
     resourceName=None
     resourceObject=/BeastGirlStormClawsUI/HUD/Materials/MI_UI_BentBaton_Input_RadialFill
     tiling=NoTile
     imageSize=<45 45>
     margin=<ltrb 0 0 0 0>
     outlineSettings {
      bUseBrushTransparency=false
      roundingType=HalfHeightRadius
      width=0
      color {colorUseRule=UseColor_Specified, specifiedColor=<rgba 0 0 0 0>}
      cornerRadii=<0 0 0 0>
     }
     tintColor {colorUseRule=UseColor_Specified, specifiedColor=<rgba 1 1 1 1>}
     uVRegion {bIsValid=false, max=<0 0>, min=<0 0>}
    }}
   slot {class=OverlaySlot, properties {horizontalAlignment=HAlign_Fill, verticalAlignment=VAlign_Fill}}
  }
  Keybind {
   widgetClassPath=/FortUILibrary/Input/KeybindWidget
   bIsVariable=true
   parent=Overlay_0
   widget=Keybind
   slot {class=OverlaySlot, properties {horizontalAlignment=HAlign_Center, verticalAlignment=VAlign_Center}}
  }
 ]
}
