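# Base HUD widget for a special abilities quickbar, wrapped in a tag visibility widget with a named slot for content.
# On Construct it initializes visibility, calls setup and intro functions that are empty here, and refreshes slots whenever input method changes.
# SetActionsToSlots picks keyboard/mouse or gamepad order, then uses data table rows to give each slot its keybind, icon, and cooldown and active tag listeners.
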WidgetBlueprint /SpecialAbilitiesQuickbarUI/WBP_SpecialAbilities_Quickbar
graphs [
 Bind Events {nodes [Bind Events() {is=title K2Node_FunctionEntry}]}
 CreateQuickbarArray {nodes [CreateQuickbarArray() {is=title K2Node_FunctionEntry}]}
 EventGraph {nodes [
   n1 = Input Method Changed!() -> (OutputDelegate, then, bNewInputType:ECommonInputType Enum) {is=title K2Node_CustomEvent}
   > SetActionsToSlots!(execute, self, InputType:ECommonInputType Enum=@n1.bNewInputType) -> (then)
   Event Construct()
   > n2 = Sequence!(execute) -> (then_0, then_1, then_2:Exec)
   > HotfixAbilityVisibility()
   > CreateQuickbarArray() {is=title K2Node_CallFunction}
   > n3 = Is Valid(InputObject=Get CommonInputSubsystem())
   n4 = Get CommonInputSubsystem()
   > Bind Event to On Input Method Changed(self=@n4, Delegate=Create Event())
   > n5 = SetActionsToSlots(InputType=GetCurrentInputType(self=@n4))
   Bind Events(execute=[@n6.then, @n5.then]) {is=title K2Node_CallFunction}
   > PlayIntroAnim()
   n6 = SetActionsToSlots(execute=@n3.Is Not Valid, InputType=MouseAndKeyboard)
   Initialize!(
    execute=@n2.then_2
    self:Common Tag Visibility Widget Object Reference=Get CommonTagVisibilityWidget() -> (CommonTagVisibilityWidget:Common Tag Visibility Widget Object Reference)
    OwningLocalPlayer:Local Player Object Reference=GetOwningLocalPlayer!(self:Widget Object Reference) -> (ReturnValue:Local Player Object Reference)
    OwningPlayerController:Player Controller Object Reference=GetOwningPlayer()
   ) -> (then)
   PlayIntroAnim()
   PlayOutroAnim()
  ]}
 HotfixAbilityVisibility {nodes [HotfixAbilityVisibility() {is=title K2Node_FunctionEntry}]}
 PlayIntroAnim {nodes [PlayIntroAnim() {is=title K2Node_FunctionEntry}]}
 PlayOutroAnim {nodes [PlayOutroAnim() {is=title K2Node_FunctionEntry}]}
 SetActionsToSlots {nodes [
   SetActionsToSlots!() -> (then, InputType:ECommonInputType Enum) {is=title K2Node_FunctionEntry}
   > n1 = For Each Loop(Array:Array of WBP Special Abilities Quickbar Slot Object References=Get QuickbarSlots() -> (QuickbarSlots:Array of WBP Special Abilities Quickbar Slot Object References)) -> (Array Element:WBP Special Abilities Quickbar Slot Object Reference)
   n2 = Get KMBAbilityOrder() -> (KMBAbilityOrder:Array of Names)
   n3 = "Get (a ref)"(
    Array:Array of Names=Select!(
     MouseAndKeyboard:Array of Names=@n2
     Gamepad:Array of Names=Get GamepadAbilityOrder() -> (GamepadAbilityOrder:Array of Names)
     Touch:Array of Names=@n2
     Count:Array of Names=@n2
     Index:ECommonInputType Enum=Get InputType() -> (InputType:ECommonInputType Enum)
    ) -> (ReturnValue:Array of Names)
    Dimension 1=@n1.Array Index
   ) -> (Output:"Name (by ref)")
   > StoreDataTableRowName!(execute, self:WBP Special Abilities Quickbar Slot Object Reference=@n1, DataTableRowName:Name=@n3) -> (then)
   > n4 = Get Data Table Row!(
    execute:Exec
    DataTable:Data Table Object Reference=Get SpecialAbilitiesDataTable() -> (SpecialAbilitiesDataTable:Data Table Object Reference)
    RowName:Name=@n3
   ) -> (then:Exec, RowNotFound:Exec, ReturnValue:Struct Special Abilities Details Structure)
   n5 = Break Struct Special Abilities Details!(Struct_SpecialAbilities_Details:"Struct Special Abilities Details Structure (by ref)"=@n4) -> (
    GamepadBoundAction_15_#1:Name
    KBMBoundAction_17_#2:Name
    AbilityIcon_18_#3:Texture 2D Object Reference
    AbilityActiveTag_6_#4:Gameplay Tag Structure
    AbilityCooldownTag_8_#5:Gameplay Tag Structure
   ) {is=title K2Node_BreakStruct}
   > Assign Keybind Action!(
    execute
    self:WBP Special Abilities Quickbar Slot Object Reference=@n1
    NewBoundAction:Name=Select(
     MouseAndKeyboard:Name=@n5.KBMBoundAction_17_#2
     Gamepad:Name=@n5
     Touch:Name=@n5.KBMBoundAction_17_#2
     Count:Name=@n5.KBMBoundAction_17_#2
     Index=Get InputType()
    ) -> (ReturnValue:Name)
   ) -> (then)
   > ListenForCooldownEvents!(
    execute
    self:WBP Special Abilities Quickbar Slot Object Reference=@n1
    CooldownGameplayTag:Gameplay Tag Structure=@n5.AbilityCooldownTag_8_#5
   ) -> (then)
   > ListenForActiveEvents!(
    execute
    self:WBP Special Abilities Quickbar Slot Object Reference=@n1
    ActiveGameplayTag:Gameplay Tag Structure=@n5.AbilityActiveTag_6_#4
   ) -> (then)
   > Set AbilityIcon!(
    execute
    AbilityIcon:Texture 2D Object Reference=@n5.AbilityIcon_18_#3
    self:WBP Special Abilities Quickbar Slot Object Reference=@n1
   ) -> (then, Output_Get:Texture 2D Object Reference)
   > Update Visuals Based on Index(self:WBP Special Abilities Quickbar Slot Object Reference=@n1)
  ]}
 UnbindEvents {nodes [UnbindEvents() {is=title K2Node_FunctionEntry}]}
]
variables [
 GamepadAbilityOrder {type="array<string>", default []}
 KMBAbilityOrder {type="array<string>", default []}
 QuickbarSlots {type="array<WBP_SpecialAbilities_QuickbarSlot_C>", default []}
 SpecialAbilitiesDataTable {type=DataTable, default=null}
]
widgets {
 info {
  inheritedWidgetCount=1
  parentClass=&FortHUDElementWidget
  rootWidgetClass=/Script/FortniteUI.CommonTagVisibilityWidget
  widgetCount=3
 }
 widgets [
  CommonTagVisibilityWidget {
   widgetClassPath=&CommonTagVisibilityWidget
   bIsVariable=true
   widget=CommonTagVisibilityWidget
   properties {hUDElementTag {gameplayTags [{tagName=HUD.Athena.HideForPostGame}]}}
  }
  NamedSlot_Root {widgetClassPath=&NamedSlot, bIsVariable=true, parent=CommonTagVisibilityWidget, widget=NamedSlot_Root, slot {class=PanelSlot}}
  TutorialHighlight {widgetClassPath=&TutorialHighlight}
 ]
}
