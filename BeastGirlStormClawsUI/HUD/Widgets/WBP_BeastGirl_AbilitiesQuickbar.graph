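# BeastGirl variant of the special abilities quickbar HUD widget, laying out two BeastGirl ability slots side by side in a grid.
# CreateQuickbarArray fills QuickbarSlots with the left and right slots, and the right slot sets its own rotation animation seed.
# PlayIntroAnim and PlayOutroAnim run short opacity fades on the root overlay.
# Bind Events, UnbindEvents and HotfixAbilityVisibility are empty, with no parent calls.
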WidgetBlueprint /BeastGirlStormClawsUI/HUD/Widgets/WBP_BeastGirl_AbilitiesQuickbar
animations [
 Anim_Intro {
  binding_count=1
  length_seconds=0.15
  object=&self:Anim_Intro
  sequence {bindings [Overlay_Root {
     id=#1
     possessed_class=Overlay
     tracks [{
       type=MovieSceneFloatTrack
       display_name=Render Opacity
       sections [{type=MovieSceneFloatSection, end_frame=3, start_frame=0, channels [None {type=MovieSceneScriptingFloatChannel, keys [[0,0],[3.3333,1]]}]}]
      }]
    }]}
 }
 Anim_Outro {
  binding_count=1
  length_seconds=0.25
  object=&self:Anim_Outro
  sequence {bindings [Overlay_Root {
     id=#2
     possessed_class=Overlay
     tracks [{
       type=MovieSceneFloatTrack
       display_name=Render Opacity
       sections [{type=MovieSceneFloatSection, end_frame=5, start_frame=0, channels [None {type=MovieSceneScriptingFloatChannel, keys [[0,1],[5,0]]}]}]
      }]
    }]}
 }
]
graphs [
 Bind Events {nodes [Bind Events() {is=title K2Node_FunctionEntry}]}
 CreateQuickbarArray {nodes [
   CreateQuickbarArray() {is=title K2Node_FunctionEntry}
   > Set QuickbarSlots(QuickbarSlots:Array of WBP Special Abilities Quickbar Slot Object References=Make Array!(
     "[0]":WBP Special Abilities Quickbar Slot Object Reference=Get QuickbarSlot_Left() -> (QuickbarSlot_Left:WBP Beast Girl Abilities Quickbar Slot Object Reference)
     "[1]":WBP Special Abilities Quickbar Slot Object Reference=Get QuickbarSlot_Right() -> (QuickbarSlot_Right:WBP Beast Girl Abilities Quickbar Slot Object Reference)
    ) -> (Array:Array of WBP Special Abilities Quickbar Slot Object References)) -> (Output_Get:Array of WBP Special Abilities Quickbar Slot Object References)
  ]}
 EventGraph {nodes [
   Event Construct()
   > Parent: Construct()
   Event PreConstruct()
  ]}
 HotfixAbilityVisibility {nodes [HotfixAbilityVisibility() {is=title K2Node_FunctionEntry}]}
 PlayIntroAnim {nodes [
   PlayIntroAnim() {is=title K2Node_FunctionEntry}
   > PlayAnimation(InAnimation=Get Anim_Intro() -> (Anim_Intro:Widget Animation Object Reference))
  ]}
 PlayOutroAnim {nodes [
   PlayOutroAnim() {is=title K2Node_FunctionEntry}
   > PlayAnimation(InAnimation=Get Anim_Outro() -> (Anim_Outro:Widget Animation Object Reference))
  ]}
 UnbindEvents {nodes [UnbindEvents() {is=title K2Node_FunctionEntry}]}
]
variables []
widgets {
 info {
  inheritedWidgetCount=0
  namedSlotCount=1
  parentClass=/SpecialAbilitiesQuickbarUI/WBP_SpecialAbilities_Quickbar.WBP_SpecialAbilities_Quickbar_C
  rootWidgetClass=null
  widgetCount=4
 }
 widgets [
  Overlay_Root {widgetClassPath=&Overlay, widget=Overlay_Root}
  GridPanel_58 {widgetClassPath=&GridPanel, parent=Overlay_Root, widget=GridPanel_58, slot {class=OverlaySlot}}
  QuickbarSlot_Left {
   widgetClassPath=&/WBP_BeastGirl_AbilitiesQuickbarSlot
   bIsVariable=true
   parent=GridPanel_58
   widget=QuickbarSlot_Left
   properties {foregroundColor {colorUseRule=UseColor_Specified, specifiedColor=<rgba 1 0 1 1>}}
   slot {class=GridSlot}
  }
  QuickbarSlot_Right {
   widgetClassPath=&WBP_BeastGirl_AbilitiesQuickbarSlot
   bIsVariable=true
   parent=GridPanel_58
   widget=QuickbarSlot_Right
   properties {rotationAnimSeed=1}
   slot {class=GridSlot, properties {column=1, padding=<ltrb 30 0 0 0>}}
  }
 ]
}
