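# Translucent UI material for the BentBaton HUD health bar and its backing in BeastGirlStormClawsUI.
# An MSDF texture minus a circle SDF, clipped and faded horizontally, forms the tinted backing, with scrolling sine scanlines.
# Skewed box SDFs form the bar, which a remapped Fill parameter splits into fill and dimmer background.
# IsBelowThreshold blends between health color gradients, and named reroutes connect the layers for compositing.
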Material /BeastGirlStormClawsUI/HUD/Materials/M_UI_BentBaton_HealthBacking
graph {
 properties {bAutomaticallySetUsageInEditor=false, blendMode=BLEND_Translucent, materialDomain=MD_UI, usedShadingModels=Default Lit}
 nodes [
  MaterialExpressionMultiply_12 = Multiply(A=Time(), B=ScalarParameter() {defaultValue=-0.005, parameterName=ScanLineSpeed}) {constB=-0.015}
  n1 = Convert(None=@MaterialExpressionMultiply_12) {
   convertInputs [{
     type=Scalar
     expressionInput {
      expression=&self:MaterialExpressionMultiply_12
      inputName=None
      mask=0
      maskA=0
      maskB=0
      maskG=0
      maskR=0
      outputIndex=0
     }
    }]
   convertOutputs [{type=Vector2}]
  }
  n2 = Add(A=TextureCoordinate(), B=@n1)
  n3 = Multiply(A=@n2) {constB=120}
  n4 = ComponentMask(None=@n2) {r=true}
  n5 = Saturate(None=Multiply(A=@n4) {constB=1.2})
  n6 = Subtract(A=@n4) {constB=0.2}
  n7 = Multiply(A=ComponentMask(None=Sine(None=@n3)) {g=true}, B=Multiply(A=OneMinus(None=@n5), B=Saturate(None=@n6)))
  n8 = Multiply(A=VectorParameter() {parameterName=LineColor, defaultValue=<rgba 0.224 0 0.1823 1>}.RGB, B=@n7)
  n9 = LinearInterpolate(A=Saturate(None=Multiply(A=@n8) {constB=6}), Alpha=NamedRerouteUsage()) {constB=0}
  LinearInterpolate(A=Add(A=@n9, B=NamedRerouteUsage()), B=@n9, Alpha=NamedRerouteUsage()) {constAlpha=1}
  LinearInterpolate(A=NamedRerouteUsage(), B=NamedRerouteUsage(), Alpha=NamedRerouteUsage()) {constA=1}
  n10 = ComponentMask(None=TextureCoordinate()) {r=true}
  n11 = Saturate(None=Multiply(A=OneMinus(None=@n10)) {constB=4})
  n12 = TextureCoordinate()
  n13 = Constant() {r=-0.056}
  n14 = Step(Y=ComponentMask(None=MaterialFunctionCall(UVs=@n12, "Rotation Angle (0-1)"=@n13) {materialFunction=&CustomRotator}.Rotated Values) {r=true}) {constX=0.825}
  n15 = MaterialFunctionCall(
   DistanceField=MaterialFunctionCall!(MSDF RGB=TextureSample() {samplerSource=SSM_Clamp_WorldGroupSettings, texture=/BeastGirlStormClawsUI/HUD/Textures/T_UI_MSDF_BentBaton_HealthBacking}.RGB) {materialFunction=/FortUILibrary/MaterialFunctions/Shapes/Modifiers/MF_UI_MSDFtoSDF}.SDF
   Threshold=Constant() {r=0.25}
  ) {materialFunction=&MF_UI_AntiAliasing}
  n16 = TextureCoordinate()
  n17 = Constant2Vector() {g=0.025, r=0.54}
  n18 = MaterialFunctionCall(
   DistanceField=MaterialFunctionCall!(UVs=Add(A=@n16, B=@n17), Radius=Constant() {r=0.3}) {materialFunction=/FortUILibrary/MaterialFunctions/Shapes/SDFs/MF_SDF_Circle}.SDF
   Smoothing=Constant() {r=0.04}
  ) {materialFunction=&MF_UI_AntiAliasing}
  n19 = Multiply(
   A=Multiply(A=Multiply(A=@n11, B=@n14), B=ScalarParameter() {defaultValue=0.925, parameterName=BackingOpacity})
   B=NamedRerouteDeclaration(None=Subtract(A=@n15.Outer, B=@n18.Inner)) {name=AA'dBacking, nodeColor=<rgba 0.6641 1 0 1>}
  ) {constA=1}
  NamedRerouteDeclaration(None=@n19) {name=BackingOpacity, nodeColor=<rgba 0.7813 0 1 1>}
  NamedRerouteDeclaration(None=Multiply(A=VectorParameter() {parameterName=BackingColor, defaultValue=<rgba 0.0612 0.004 0.0423 0>}.RGB, B=@n19)) {name=BackingColor, nodeColor=<rgba 0 1 0.4609 1>}
  NamedRerouteDeclaration(None=@n7) {name=ScanLines, nodeColor=<rgba 0 0.7422 1 1>}
  n20 = VectorParameter() {parameterName=FillBackgroundColor, defaultValue=<rgba 0 0 0 1>}
  n21 = NamedRerouteUsage()
  n22 = NamedRerouteUsage()
  NamedRerouteDeclaration(None=LinearInterpolate(A=NamedRerouteUsage(), B=LinearInterpolate(A=@n20.RGB, B=@n21, Alpha=@n22), Alpha=NamedRerouteUsage()) {constAlpha=1}) {name=CompositedHealthBar, nodeColor=<rgba 1 0.9375 0 1>}
  n23 = MaterialFunctionCall!(UV=Add(A=TextureCoordinate(), B=Constant2Vector()), Skew Angle=Constant2Vector() {r=20}) {materialFunction=/FortUILibrary/MaterialFunctions/Basics/MF_UI_Skew}
  n24 = ScalarParameter() {defaultValue=1, parameterName=Fill}
  n25 = Constant2Vector() {g=1}
  n26 = Constant2Vector() {g=0.82, r=0.118}
  n27 = MaterialFunctionCall(UV=Add(A=TextureCoordinate(), B=Constant2Vector() {g=0.02, r=0.025}), Skew Angle=Constant2Vector() {r=20}) {materialFunction=&MF_UI_Skew}
  n28 = Constant2Vector() {g=0.03, r=0.7}
  n29 = Add(A=TextureCoordinate(), B=Constant2Vector() {g=0.025, r=0.52})
  n30 = Constant() {r=0.3}
  n31 = MaterialFunctionCall(DistanceField=MaterialFunctionCall!(
    DistanceFieldA=MaterialFunctionCall!(UVs=@n27.Skewed UVs, Size=@n28, CornerRoundness, KeepAspectRatio, PixelSize) {materialFunction=/FortUILibrary/MaterialFunctions/Shapes/SDFs/MF_SDF_Box}.SDF
    DistanceFieldB=MaterialFunctionCall(UVs=@n29, Size=@n30) {materialFunction=&MF_SDF_Box}.SDF
    BlendAmount=Constant()
   ) {materialFunction=/FortUILibrary/MaterialFunctions/Shapes/Modifiers/MF_UI_SDF_Subtraction}.Exact) {materialFunction=&MF_UI_AntiAliasing}
  n32 = Multiply(
   A=Step(
    Y=ComponentMask(None=@n23.Skewed UVs) {r=true}
    X=MaterialFunctionCall(Input=@n24, Input Low=@n25.R, Input High=@n25.R, Target Low=@n26.R, Target High=@n26.R) {materialFunction=&RemapValueRange}.Result
   ) {constX=0.5}
   B=NamedRerouteDeclaration(None=@n31.Inner) {name=FullFillBar, nodeColor=<rgba 1 0 0.9844 1>}
  )
  n33 = Multiply(A=Subtract(A=@n31.Inner, B=NamedRerouteUsage())) {constB=0.75}
  NamedRerouteDeclaration(None=Add(
    A=NamedRerouteDeclaration(None=@n32) {name=FillColor, nodeColor=<rgba 1 0.1406 0 1>}
    B=NamedRerouteDeclaration(None=@n33) {name=FillBackground, nodeColor=<rgba 0 0.1797 1 1>}
   )) {name=CombinedFill, nodeColor=<rgba 0 0.7422 1 1>}
  n34 = VectorParameter() {parameterName=ColorA, defaultValue=<rgba 0 0.2874 0.0595 1>}
  n35 = VectorParameter() {parameterName=ColorB, defaultValue=<rgba 0.1384 0.7231 0.0203 1>}
  n36 = TextureCoordinate()
  n37 = Saturate(None=ConstantBiasScale(None=ComponentMask(None=@n36) {r=true}) {bias=-0.02, scale=1.3})
  NamedRerouteDeclaration(None=LinearInterpolate(A=NamedRerouteUsage(), B=LinearInterpolate(A=@n34.RGB, B=@n35.RGB, Alpha=@n37), Alpha=ScalarParameter() {parameterName=IsBelowThreshold})) {name=HealthColor, nodeColor=<rgba 0.7578 0 1 1>}
  n38 = TextureCoordinate()
  n39 = Multiply(A=Saturate(None=ComponentMask(None=@n38) {r=true})) {constB=3}
  n40 = LinearInterpolate(
   A=VectorParameter() {parameterName=HealthBarColor1, defaultValue=<rgba 0.7084 0.0887 0.5271 1>}.RGB
   B=VectorParameter() {parameterName=HealthBarColor2, defaultValue=<rgba 0.7835 0.0887 0.5271 1>}.RGB
   Alpha=Saturate(None=@n39)
  )
  n41 = VectorParameter() {parameterName=HealthBarColor3, defaultValue=<rgba 1 0.4508 0.8469 1>}
  n42 = Subtract(A=@n39) {constB=1.3}
  n43 = Saturate(None=@n42)
  n44 = Subtract(A=@n42)
  NamedRerouteDeclaration(None=LinearInterpolate(
    A=LinearInterpolate(A=@n40, B=@n41.RGB, Alpha=@n43)
    B=VectorParameter() {parameterName=HealthBarColor4, defaultValue=<rgba 1 1 1 1>}.RGB
    Alpha=Saturate(None=@n44)
   )) {name=HealthAboveDurabilityThreshold, nodeColor=<rgba 0.5937 1 0 1>}
 ]
}
