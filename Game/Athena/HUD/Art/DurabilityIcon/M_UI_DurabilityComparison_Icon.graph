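# Translucent UI material for the durability comparison icon, drawn from distance fields in a channel-packed texture.
# The red-channel main shape shifts horizontally when EnableComparison is set; the green-channel comparison shape has adjustable offset, rotation, scale and outline.
# EnableComparison also gates a mask that blends MainColor to ComparisonColor, and a subtract-then-blend that composites the shape masks.
# The isArrow blend takes green for both inputs, so it changes nothing.
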Material /Game/Athena/HUD/Art/DurabilityIcon/M_UI_DurabilityComparison_Icon
graph {
 properties {bAutomaticallySetUsageInEditor=false, blendMode=BLEND_Translucent, materialDomain=MD_UI, usedShadingModels=Default Lit}
 nodes [
  n1 = NamedRerouteUsage()
  LinearInterpolate(
   A=VectorParameter() {parameterName=MainColor, defaultValue=<rgba 1 1 1 1>}.RGB
   B=VectorParameter() {parameterName=ComparisonColor, defaultValue=<rgba 1 0 0 1>}.RGB
   Alpha=LinearInterpolate(B=Ceil(None=@n1), Alpha=ScalarParameter() {parameterName=EnableComparison})
  )
  n2 = NamedRerouteUsage()
  n3 = LinearInterpolate(B=NamedRerouteUsage(), Alpha=ScalarParameter() {parameterName=EnableComparison})
  n4 = LinearInterpolate(B=NamedRerouteUsage(), Alpha=ScalarParameter() {parameterName=EnableComparison})
  LinearInterpolate(A=Saturate(None=Subtract(A=@n2, B=@n3)), B=@n4, Alpha=@n4)
  n5 = TextureCoordinate()
  n6 = ScalarParameter() {defaultValue=0.08, parameterName=DiamondXAxis}
  n7 = ScalarParameter() {parameterName=EnableComparison}
  n8 = AppendVector(A=Add(A=ComponentMask(None=@n5) {r=true}, B=LinearInterpolate(B=@n6, Alpha=@n7)), B=ComponentMask(None=@n5) {g=true})
  NamedRerouteDeclaration(None=MaterialFunctionCall!(
    RGBA Channels=TextureSample(UVs=@n8) {texture=/Game/Athena/HUD/Art/T_UI_Icon_DurabilityChannels}.R
    Glow Min=Constant() {r=0.3}
    Glow Max=Constant() {r=0.5}
    Outline Thickness
   ) {materialFunction=/FortUILibrary/MaterialFunctions/Basics/MF_UI_ChannelPackedDistanceField}.Fill) {name=MainShape, nodeColor=<rgba 0 0 0 1>}
  n9 = TextureCoordinate()
  n10 = AppendVector(
   A=Add(A=ComponentMask(None=@n9) {r=true}, B=ScalarParameter() {parameterName=XAxis})
   B=Add(A=ComponentMask(None=@n9) {g=true}, B=ScalarParameter() {parameterName=YAxis})
  )
  n11 = ScalarParameter() {parameterName=Rotation}
  n12 = TextureSample(UVs=MaterialFunctionCall(
    UVs=MaterialFunctionCall(UVs=@n10, "Rotation Angle (0-1)"=@n11) {materialFunction=&CustomRotator}.Rotated Values
    Texture Scale=ScalarParameter() {defaultValue=1, parameterName=Scale}
   ) {materialFunction=&ScaleUVsByCenter}.UVs) {samplerSource=SSM_Clamp_WorldGroupSettings, texture=&T_UI_Icon_DurabilityChannels}
  n13 = MaterialFunctionCall(
   RGBA Channels=LinearInterpolate(A=@n12.G, B=@n12.G, Alpha=ScalarParameter() {parameterName=isArrow})
   Glow Min=Constant() {r=0.3}
   Glow Max=Constant() {r=0.5}
   Outline Thickness=ScalarParameter() {defaultValue=0.7831, parameterName=OutlineThickness}
  ) {materialFunction=&MF_UI_ChannelPackedDistanceField}
  NamedRerouteDeclaration(None=@n13.Fill) {name=Comparison, nodeColor=<rgba 0 0 0 1>}
  NamedRerouteDeclaration(None=@n13.Outline) {name=ComparisonOutline, nodeColor=<rgba 0 0 0 1>}
 ]
}
