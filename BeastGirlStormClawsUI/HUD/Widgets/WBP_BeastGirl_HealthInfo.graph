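# HUD widget showing Beast Girl Storm Claws health as a colored bar and durability percentage, scaled for mobile touch.
# Pawn health changes restart a timer-driven ease-out that writes fill to a dynamic material, with colors chosen by low-health thresholds.
# The reverse-transformation tag, or leap-blocked while spectating, instead deducts the curve-table transform durability cost from displayed health.
# An out-of-health tag on the stand-in pawn's PlayerState zeroes the display and marks it dead.
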WidgetBlueprint /BeastGirlStormClawsUI/HUD/Widgets/WBP_BeastGirl_HealthInfo
graphs [
 EventGraph {nodes [
   OnHealthChanged() {is=title K2Node_CustomEvent}
   > Branch(Condition=InvertBool!(bValue:bool=Get bIsDead() -> (bIsDead:bool)) -> (ReturnValue:bool))
   > n1 = Is Spectating!(execute:Exec) -> (true:Exec, false:Exec) {is=title K2Node_MacroInstance}
   DelayUntilNextTick!(execute=@n1.false) -> (then)
   > n2 = HasMatchingGameplayTag!(
    execute:Exec
    self:Object Reference=Get CachedViewTarget() -> (CachedViewTarget:Actor Object Reference)
    TagToCheck:Gameplay Tag Structure=Get ReverseTransformationTag() -> (ReverseTransformationTag:Gameplay Tag Structure)
   ) -> (then:Exec, ReturnValue:bool) {is=title K2Node_Message}
   > n3 = Branch(Condition=NOT Boolean(A=@n2))
   n4 = GetPawnHealth!(execute=[@n3.then, @n6.else], self) -> (then, Value:double)
   > Set CachedBeastGirlHealth(CachedBeastGirlHealth:double=@n4) -> (Output_Get:double)
   > SetNewHealth!(execute, self, HealthPercentage:double=@n4) -> (then)
   > SetDurabilityPercentage!(execute, self, IsDead:bool=false) -> (then)
   > SetHealthBarColors()
   n5 = HasMatchingGameplayTag(execute=@n1.true, self=Get CachedViewTarget(), TagToCheck=Get LeapBlockedTag() -> (LeapBlockedTag:Gameplay Tag Structure))
   > n6 = Branch(Condition=@n5)
   > SetReverseTransformationHealth()
   ListenForReverseTransformation() {is=title K2Node_CustomEvent}
   > n7 = WaitGameplayTagAddToActor!(
    execute:Exec
    TargetActor:Actor Object Reference=GetViewTarget!(self:Controller Object Reference=GetOwningPlayer()) -> (ReturnValue:Actor Object Reference)
    Tag:Gameplay Tag Structure=Get ReverseTransformationTag()
    OnlyTriggerOnce:bool=false
   ) -> (then:Exec, AsyncTaskProxy:Ability Async Wait Gameplay Tag Added Object Reference, Added:Exec) {is=title K2Node_AsyncAction}
   SetReverseTransformationHealth(execute=@n7.Added)
   Event PreConstruct()
   > SetMobileStyle() {is=title K2Node_CallFunction}
   Event Construct()
   > Set bIsDead(bIsDead:bool=false) -> (Output_Get:bool)
   > ListenForReverseTransformation() {is=title K2Node_CallFunction}
   > Set CachedViewTarget(CachedViewTarget:Actor Object Reference=GetViewTarget(self=GetOwningPlayer())) -> (Output_Get:Actor Object Reference)
   > n8 = GetDynamicMaterial(self=Get Image_HealthBar() -> (Image_HealthBar:Fort Lazy Image Object Reference))
   > Set CachedHealthBarMaterial(CachedHealthBarMaterial:Material Instance Dynamic Object Reference=@n8) -> (Output_Get:Material Instance Dynamic Object Reference)
   > SetupHealthBar()
   > SetDurabilityPercentage(IsDead=false)
   > SetHealthBarColors()
   > n9 = GetPawnHealth()
   > Set InitialHealthOnTransform(InitialHealthOnTransform:double=@n9) -> (Output_Get:double)
   > n10 = Cast To BeastGirlStandInPawn!(execute, Object=Get CachedViewTarget()) -> (then, CastFailed, AsBeast Girl Stand in Pawn:Beast Girl Stand In Pawn Object Reference) {is=title K2Node_DynamicCast}
   > n11 = WaitGameplayTagAddToActor(
    TargetActor=Get PlayerState!(self:Pawn Object Reference=@n10) -> (PlayerState:Player State Object Reference)
    Tag=(TagName="Granted.BeastGirl.OutOfHealth")
    OnlyTriggerOnce=false
   )
   SetNewHealth(execute=@n11.Added, HealthPercentage=0)
   > SetDurabilityPercentage(IsDead=true)
   > Set bIsDead(bIsDead=true)
  ]}
 GetPawnHealth {nodes [
   GetPawnHealth() {is=title K2Node_FunctionEntry}
   > n1 = Cast To FortPlayerPawn(Object=GetViewTarget(self=GetOwningPlayer()))
   n2 = GetMaxHealth!(self:Fort Pawn Object Reference=@n1) -> (ReturnValue:float)
   > Return Node!(execute, Value:double=float / float(A=GetHealth!(self:Fort Pawn Object Reference=@n1) -> (ReturnValue:float), B="Max (Float)"(A=@n2, B=1)))
   Return Node(execute=@n1.CastFailed, Value=0)
  ]}
 GetReverseTransformationHealth {nodes [
   GetReverseTransformationHealth!() -> (then, BeastGirlHealth:float) {is=title K2Node_FunctionEntry}
   > n1 = Cast To FortPlayerPawn(Object=Get CachedViewTarget())
   > Set BeastGirlMaxHealth(BeastGirlMaxHealth:double=GetMaxHealth(self=@n1)) -> (Output_Get:double)
   n2 = Get HF_DurabilityDecreasePerTransform() -> (HF_DurabilityDecreasePerTransform:Scalable Float Structure)
   > Set HealthDecreasePerTransform(HealthDecreasePerTransform:double=float / float(
     A=GetValueAtLevel!(Input:"Scalable Float Structure (by ref)"=@n2, Level:float=0) -> (ReturnValue:float)
     B=Get BeastGirlMaxHealth() -> (BeastGirlMaxHealth:double)
    )) -> (Output_Get:double)
   n3 = Get HealthDecreasePerTransform() -> (HealthDecreasePerTransform:double)
   n4 = Get BeastGirlHealth() -> (BeastGirlHealth:float)
   n5 = float - float(A=Get InitialHealthOnTransform() -> (InitialHealthOnTransform:double), B=Get HealthDecreasePerTransform())
   n6 = SelectFloat(A=0, B=SelectFloat(A=0, B=100, bPickA="float <= float"(A=@n5) -> (ReturnValue=~)), bPickA=Get bIsDead())
   n7 = Get BeastGirlMaxHealth()
   > Return Node!(execute, HealthPercentage:double="Clamp (Float)"(Value=float - float(A=@n4, B=@n3), Min=float / float(A=@n6, B=@n7), Max=1))
   Return Node(execute=@n1.CastFailed, HealthPercentage=0)
  ]}
 InterpolateHealth {nodes [
   InterpolateHealth() {is=title K2Node_FunctionEntry}
   > n1 = Sequence!(execute) -> (then_0, then_1, then_2:Exec)
   n2 = float * float(A=GetWorldDeltaSeconds!() -> (ReturnValue:double), B=Get HealthInterpolationRate() -> (HealthInterpolationRate:double)) {is=title K2Node_PromotableOperator}
   n3 = Get InterpolationTimeStep() -> (InterpolationTimeStep:double)
   > n4 = Set InterpolationTimeStep(InterpolationTimeStep:double="Min (Float)"(A:double=float + float(A=@n3, B=@n2), B:double=1) -> (ReturnValue:double) {is=title K2Node_CommutativeAssociativeBinaryOperator}) -> (Output_Get:double)
   > Set InterpolatedHealthValue(InterpolatedHealthValue:double=Ease!(
     Function:EEasingFunc Enum=EaseOut
     Alpha:double=@n4
     A:double=Get InterpolationStartValue() -> (InterpolationStartValue:double)
     B:double=Get CurrentHealth() -> (CurrentHealth:double)
     BlendExp:double=2
    ) -> (Result:double=~) {is=title K2Node_EaseFunction}) -> (Output_Get:double)
   Branch(execute=@n1.then_1, Condition="Nearly Equal (Float)"(A=Get InterpolationTimeStep(), B=1, ErrorTolerance=0.001))
   > Clear and Invalidate Timer by Handle(Handle=Get HealthUpdateHandle() -> (HealthUpdateHandle:Timer Handle Structure))
   > Set InterpolatedHealthValue(InterpolatedHealthValue=Get CurrentHealth())
   SetScalarParameterValue(
    execute=@n1.then_2
    self=Get CachedHealthBarMaterial() -> (CachedHealthBarMaterial:Material Instance Dynamic Object Reference)
    ParameterName=Get HealthMaterialParam() -> (HealthMaterialParam:Name=None)
    Value=Get InterpolatedHealthValue() -> (InterpolatedHealthValue:double)
   )
  ]}
 SetDurabilityPercentage {nodes [
   SetDurabilityPercentage!() -> (then, IsDead:bool=~) {is=title K2Node_FunctionEntry}
   n1 = float * float(A=Get CurrentHealth(), B=100)
   n2 = Round to Integer64!(A:double=@n1) -> (ReturnValue:Integer64=0)
   n3 = AND Boolean(A="float > float"(A=@n1), B="float < float"(A=@n1, B=1))
   n4 = Ceil to Integer64!(A:double=@n1) -> (ReturnValue:Integer64=0)
   n5 = "To String (Integer64)"!(InInt:Integer64=Select(Option 0:Integer64=Select(Option 0=@n2, Option 1=@n4, Index=@n3), Option 1:Integer64=0, Index=Get IsDead() -> (IsDead:bool)) -> (ReturnValue:Integer64)) -> (ReturnValue:String)
   > Set Text(
    Text="To Text (String)"(InString=Append(A=@n5, B=%))
    self=Get CommonText_Durability() -> (CommonText_Durability:Common Text Object Reference)
   )
  ]}
 SetHealthBarColors {nodes [
   SetHealthBarColors() {is=title K2Node_FunctionEntry}
   n1 = Get LowHealthThreshold() -> (LowHealthThreshold:double)
   n2 = Get CurrentHealth()
   > SetScalarParameterValue(self=Get CachedHealthBarMaterial(), ParameterName=IsBelowThreshold, Value=SelectFloat(A=0, B=1, bPickA="float > float"(A=@n2, B=@n1)))
   n3 = "float <= float"(A=Get CurrentHealth(), B=Get LowHealthThreshold()) -> (ReturnValue=~)
   n4 = Get ColorAMidDurability() -> (ColorAMidDurability:Linear Color Structure)
   n5 = Get ColorAHighDurability() -> (ColorAHighDurability:Linear Color Structure)
   n6 = Get CurrentHealth()
   n7 = Get VeryLowHealthThreshold() -> (VeryLowHealthThreshold:double)
   > SetVectorParameterValue(
    self=Get CachedHealthBarMaterial()
    ParameterName=ColorA
    Value=SelectColor(
     A=Get ColorALowDurability() -> (ColorALowDurability:Linear Color Structure)
     B=SelectColor(A=@n4, B=@n5, bPickA=@n3)
     bPickA="float < float"(A=@n6, B=@n7)
    )
   )
   n8 = "float < float"(A=Get CurrentHealth(), B=Get LowHealthThreshold())
   n9 = Get ColorBHighDurability() -> (ColorBHighDurability:Linear Color Structure)
   n10 = Get ColorBMidDurability() -> (ColorBMidDurability:Linear Color Structure)
   n11 = Get CurrentHealth()
   n12 = Get VeryLowHealthThreshold()
   > SetVectorParameterValue(
    self=Get CachedHealthBarMaterial()
    ParameterName=ColorB
    Value=SelectColor(
     A=Get ColorBLowDurability() -> (ColorBLowDurability:Linear Color Structure)
     B=SelectColor(A=@n10, B=@n9, bPickA=@n8)
     bPickA="float < float"(A=@n11, B=@n12)
    )
   )
  ]}
 SetMobileStyle {nodes [
   SetMobileStyle() {is=title K2Node_FunctionEntry}
   n1 = GetOwningPlayer()
   > Branch(Condition=AND Boolean(A=IsMobileGame(), B=IsUsingTouch(PlayerController=@n1)))
   > Set UserSpecifiedScale!(
    execute
    UserSpecifiedScale:float=1.5
    self:Scale Box Object Reference=Get ScaleBox_HealthInfo() -> (ScaleBox_HealthInfo:Scale Box Object Reference)
   ) -> (then, Output_Get:float)
  ]}
 SetNewHealth {nodes [
   n1 = SetNewHealth!() -> (then, HealthPercentage:double=~) {is=title K2Node_FunctionEntry}
   > Set CurrentHealth(CurrentHealth:double=@n1) -> (Output_Get:double)
   > Set InterpolationStartValue(InterpolationStartValue:double=Get InterpolatedHealthValue()) -> (Output_Get:double)
   > Set InterpolationTimeStep(InterpolationTimeStep=0)
   > n2 = Branch(Condition=Is Valid Timer Handle!(Handle:Timer Handle Structure=Get HealthUpdateHandle()) -> (ReturnValue:bool))
   n3 = Set Timer by Event(
    execute=@n2.else
    Delegate=Create Event()
    Time=Get HealthUpdateInterval() -> (HealthUpdateInterval:double)
    bLooping=true
    bMaxOncePerFrame=true
   )
   > Set HealthUpdateHandle(HealthUpdateHandle:Timer Handle Structure=@n3) -> (Output_Get:Timer Handle Structure)
  ]}
 SetReverseTransformationHealth {nodes [
   SetReverseTransformationHealth() {is=title K2Node_FunctionEntry}
   > n1 = GetReverseTransformationHealth!(execute, self, BeastGirlHealth:float=Get CurrentHealth()) -> (then, HealthPercentage:double) {is=title K2Node_CallFunction}
   > SetNewHealth(HealthPercentage=@n1) {is=title K2Node_CallFunction}
   > SetDurabilityPercentage(IsDead=false) {is=title K2Node_CallFunction}
   > SetHealthBarColors() {is=title K2Node_CallFunction}
   Get CachedBeastGirlHealth() -> (CachedBeastGirlHealth:double)
  ]}
 SetupHealthBar {nodes [
   SetupHealthBar() {is=title K2Node_FunctionEntry}
   > n1 = GetPawnHealth() {is=title K2Node_CallFunction}
   > n2 = Set CurrentHealth(CurrentHealth=@n1)
   > n3 = Set CachedBeastGirlHealth(CachedBeastGirlHealth=@n2)
   > n4 = Set InterpolatedHealthValue(InterpolatedHealthValue=@n3)
   > SetScalarParameterValue(self=Get CachedHealthBarMaterial(), ParameterName=Get HealthMaterialParam() -> (HealthMaterialParam=None), Value=@n4)
   > n5 = Cast To FortPlayerPawn(Object=GetViewTarget(self=GetOwningPlayer()))
   > Bind Event to On Pawn Health Changed(self:Fort Pawn Object Reference=@n5, Delegate=Create Event()) {is=title K2Node_AddDelegate}
   > SetHealthBarColors()
  ]}
]
variables [
 BeastGirlTransformationManager {type=B_BeastGIrlTransformationManager_C, default=null}
 bIsDead {type=boolean}
 CachedBeastGirlHealth {type=number}
 CachedHealthBarMaterial {type=MaterialInstanceDynamic}
 CachedViewTarget {type=Actor, default=null}
 ColorAHighDurability {type=LinearColor, default=<rgba 0 0.3176 0.0314 1>}
 ColorALowDurability {type=LinearColor, default=<rgba 1 0.0314 0.0314 1>}
 ColorAMidDurability {type=LinearColor, default=<rgba 1 0.2275 0.0314 1>}
 ColorBHighDurability {type=LinearColor, default=<rgba 0.1333 0.6039 0.0314 1>}
 ColorBLowDurability {type=LinearColor, default=<rgba 1 0.1255 0.0314 1>}
 ColorBMidDurability {type=LinearColor, default=<rgba 1 0.6039 0 1>}
 CurrentHealth {type=number}
 HealthInterpolationRate {type=number, default=2.5}
 HealthMaterialParam {type=string, default=Fill}
 HealthUpdateHandle {type=TimerHandle}
 HealthUpdateInterval {type=number, default=0.06}
 HF_DurabilityDecreasePerTransform {
  type=ScalableFloat
  default {
   registryType=None
   value=1
   curve {curveTable=/BeastGirlStormClawsGameplay/DataTables/BeastGirlStormClawsGameplayGameData, rowName=BeastGirl.Transform.DurabilityCost}
  }
 }
 InitialHealthOnTransform {type=number}
 InterpolatedHealthValue {type=number}
 InterpolationStartValue {type=number}
 InterpolationTimeStep {type=number}
 LeapBlockedTag {type=GameplayTag, default {tagName=Ability.BeastGirl.StormClaws.Leap.Blocked}}
 LowHealthThreshold {type=number, default=0.5}
 OwningPlayer {type=PlayerController, default=null}
 ReverseTransformationTag {type=GameplayTag, default {tagName=Gameplay.Action.BeastGirl.TransformAbility.Ended}}
 VeryLowHealthThreshold {type=number, default=0.25}
]
widgets {
 info {inheritedWidgetCount=1, parentClass=&FortHUDElementWidget, rootWidgetClass=&Overlay, widgetCount=13}
 widgets [
  Overlay_Root {widgetClassPath=&Overlay, widget=Overlay_Root}
  ScaleBox_HealthInfo {
   widgetClassPath=&ScaleBox
   bIsVariable=true
   parent=Overlay_Root
   widget=ScaleBox_HealthInfo
   properties {stretch=UserSpecified, userSpecifiedScale=1.75}
   slot {class=OverlaySlot, properties {horizontalAlignment=HAlign_Fill, verticalAlignment=VAlign_Center}}
  }
  GridPanel_0 {
   widgetClassPath=&GridPanel
   parent=ScaleBox_HealthInfo
   widget=GridPanel_0
   properties {bOverride_Cursor=true}
   slot {class=ScaleBoxSlot}
  }
  Overlay_StarIcon {widgetClassPath=&Overlay, parent=GridPanel_0, widget=Overlay_StarIcon, slot {class=GridSlot, properties {layer=1}}}
  Image_StarIcon {
   widgetClassPath=&Image
   bIsVariable=true
   parent=Overlay_StarIcon
   widget=Image_StarIcon
   properties {brush {
     bIsDynamicallyLoaded=false
     drawAs=Image
     imageType=FullColor
     mirroring=NoMirror
     resourceName=None
     resourceObject=/BeastGirlStormClawsUI/HUD/Materials/M_UI_BentBaton_HUD_Icon
     tiling=NoTile
     imageSize=<80 80>
     margin=<ltrb 0 0 0 0>
     outlineSettings {
      bUseBrushTransparency=false
      roundingType=HalfHeightRadius
      width=0
      color {colorUseRule=UseColor_Specified, specifiedColor=<rgba 0 0 0 0>}
      cornerRadii=<0 0 0 0>
     }
     tintColor {colorUseRule=UseColor_Specified, specifiedColor=<rgba 1 1 1 1>}
     uVRegion {bIsValid=false, max=<0 0>, min=<0 0>}
    }}
   slot {class=OverlaySlot}
  }
  Image_Glow {
   widgetClassPath=&Image
   bIsVariable=true
   parent=Overlay_StarIcon
   widget=Image_Glow
   properties {
    brush {
     bIsDynamicallyLoaded=false
     drawAs=Image
     imageType=FullColor
     mirroring=NoMirror
     resourceName=None
     resourceObject=/BeastGirlStormClawsUI/HUD/Materials/M_UI_BentBaton_Glow
     tiling=NoTile
     imageSize=<80 80>
     margin=<ltrb 0 0 0 0>
     outlineSettings {
      bUseBrushTransparency=false
      roundingType=HalfHeightRadius
      width=0
      color {colorUseRule=UseColor_Specified, specifiedColor=<rgba 0 0 0 0>}
      cornerRadii=<0 0 0 0>
     }
     tintColor {colorUseRule=UseColor_Specified, specifiedColor=<rgba 1 1 1 1>}
     uVRegion {bIsValid=false, max=<0 0>, min=<0 0>}
    }
    renderTransform {angle=0, scale=<1.25 1.25>, shear=<0 0>, translation=<0 0>}
   }
   slot {class=OverlaySlot, properties {horizontalAlignment=HAlign_Center, verticalAlignment=VAlign_Center}}
  }
  Image_HealthBar {
   widgetClassPath=&FortLazyImage
   bIsVariable=true
   parent=GridPanel_0
   widget=Image_HealthBar
   properties {
    brush {
     bIsDynamicallyLoaded=false
     drawAs=Box
     imageType=FullColor
     mirroring=NoMirror
     resourceName=None
     resourceObject=/BeastGirlStormClawsUI/HUD/Materials/M_UI_BentBaton_HealthBacking
     tiling=NoTile
     imageSize=<290 32>
     margin=<ltrb 0 0 0 0>
     outlineSettings {
      bUseBrushTransparency=false
      roundingType=HalfHeightRadius
      width=0
      color {colorUseRule=UseColor_Specified, specifiedColor=<rgba 0 0 0 0>}
      cornerRadii=<0 0 0 0>
     }
     tintColor {colorUseRule=UseColor_Specified, specifiedColor=<rgba 1 1 1 1>}
     uVRegion {bIsValid=false, max=<0 0>, min=<0 0>}
    }
    renderTransform {angle=0, scale=<1 12>, shear=<0 0>, translation=<0 0>}
   }
   slot {class=GridSlot, properties {column=1, horizontalAlignment=HAlign_Center, verticalAlignment=VAlign_Center, nudge=<-47 6>}}
  }
  Image_TechDecor {
   widgetClassPath=&Image
   bIsVariable=true
   parent=GridPanel_0
   widget=Image_TechDecor
   properties {brush {
     bIsDynamicallyLoaded=false
     drawAs=Image
     imageType=FullColor
     mirroring=NoMirror
     resourceName=None
     resourceObject=/BeastGirlStormClawsUI/HUD/Materials/M_UI_BentBaton_HealthTopper
     tiling=NoTile
     imageSize=<32 32>
     margin=<ltrb 0 0 0 0>
     outlineSettings {
      bUseBrushTransparency=false
      roundingType=HalfHeightRadius
      width=0
      color {colorUseRule=UseColor_Specified, specifiedColor=<rgba 0 0 0 0>}
      cornerRadii=<0 0 0 0>
     }
     tintColor {colorUseRule=UseColor_Specified, specifiedColor=<rgba 1 1 1 1>}
     uVRegion {bIsValid=false, max=<0 0>, min=<0 0>}
    }}
   slot {class=GridSlot, properties {column=1, horizontalAlignment=HAlign_Left, layer=1, verticalAlignment=VAlign_Center, nudge=<-9.25 -20>}}
  }
  Overlay_Durability {
   widgetClassPath=&Overlay
   parent=GridPanel_0
   widget=Overlay_Durability
   properties {renderTransformPivot=<0 0.5>}
   slot {class=GridSlot, properties {column=1, horizontalAlignment=HAlign_Left, layer=1, verticalAlignment=VAlign_Bottom, nudge=<-9.3 -17.33>}}
  }
  GridPanel_4 {
   widgetClassPath=&GridPanel
   parent=Overlay_Durability
   widget=GridPanel_4
   properties {bOverride_Cursor=true, renderTransform {angle=0, scale=<0.6 0.6>, shear=<0 0>, translation=<0 0>}, renderTransformPivot=<0 0.5>}
   slot {class=OverlaySlot, properties {verticalAlignment=VAlign_Center}}
  }
  Image_DurabilityIcon {
   widgetClassPath=&Image
   bIsVariable=true
   parent=GridPanel_4
   widget=Image_DurabilityIcon
   properties {
    visibility=Collapsed
    brush {
     bIsDynamicallyLoaded=false
     drawAs=Image
     imageType=FullColor
     mirroring=NoMirror
     resourceName=None
     resourceObject=/Game/Athena/HUD/Art/DurabilityIcon/M_UI_DurabilityComparison_Icon
     tiling=NoTile
     imageSize=<17.3 17.3>
     margin=<ltrb 0 0 0 0>
     outlineSettings {
      bUseBrushTransparency=false
      roundingType=HalfHeightRadius
      width=0
      color {colorUseRule=UseColor_Specified, specifiedColor=<rgba 0 0 0 0>}
      cornerRadii=<0 0 0 0>
     }
     tintColor {colorUseRule=UseColor_Specified, specifiedColor=<rgba 1 1 1 1>}
     uVRegion {bIsValid=false, max=<0 0>, min=<0 0>}
    }
   }
   slot {class=GridSlot, properties {horizontalAlignment=HAlign_Left, verticalAlignment=VAlign_Center}}
  }
  CommonText_Durability {
   widgetClassPath=&CommonTextBlock
   bIsVariable=true
   parent=GridPanel_4
   widget=CommonText_Durability
   properties {
    mobileFontSizeMultiplier=1.2
    style=/FortUILibrary/Styles/Typography/Foundation/TextStyle-BurbankSmall-XS-White.TextStyle-BurbankSmall-XS-White_C
    text=100%
    font {
     bForceMonospaced=false
     bMaterialIsStencil=false
     fontMaterial=null
     fontObject=&BurbankSmall
     letterSpacing=0
     monospacedWidth=1
     size=13
     skewAmount=0
     typefaceFontName=Black
     outlineSettings {
      bApplyOutlineToDropShadows=false
      bMiteredCorners=false
      bSeparateFillAlpha=false
      outlineMaterial=null
      outlineSize=0
      outlineColor=<rgba 0 0 0 1>
     }
    }
    shadowOffset=<0 0>
   }
   slot {class=GridSlot, properties {column=1, horizontalAlignment=HAlign_Left, verticalAlignment=VAlign_Center}}
  }
  TutorialHighlight {widgetClassPath=&TutorialHighlight}
 ]
}
